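# Idle Niagara effect for the PixieParts character body, adding glows, lights and energy ribbons attached to its skeletal mesh.
# Each emitter bursts persistent particles, using partition indices to pick bones and offsets every update.
# Trails place four four-point ribbons along sampled skeleton curves, while Glows and Lights follow individual bones.
# System color variables and a fade-in drive color, with front-end modifiers adjusting brightness and glow offsets versus gameplay.
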NiagaraSystem /BRCosmetics/Characters/Player/Male/Medium/Bodies/M_MED_PixieParts_Lava/FX/NS_Character_PixieParts_Idle
dependencies {
 usedDataInterfaces [&NiagaraDataInterfaceCurve, &NiagaraDataInterfaceArrayInt32, &NiagaraDataInterfaceSkeletalMesh]
 usedDynamicInputs [
  &FloatFromCurve
  &MergeHelper_ColorSwitch
  &MakeIntFromFloat
  &FNTimeOfDayColor
 ]
 usedModules [
  &SystemState
  &self:SystemSpawnScript.SystemScriptSource.SystemScriptGraph.NiagaraNodeAssignment_1.SetVariables_#1
  &self:SystemSpawnScript.SystemScriptSource.SystemScriptGraph.NiagaraNodeAssignment_2.SetVariables_#2
  &self:SystemSpawnScript.SystemScriptSource.SystemScriptGraph.NiagaraNodeAssignment_3.SetVariables_#3
  &self:SystemSpawnScript.SystemScriptSource.SystemScriptGraph.NiagaraNodeAssignment_4.SetVariables_#4
  &self:SystemSpawnScript.SystemScriptSource.SystemScriptGraph.NiagaraNodeAssignment_0.SetVariables_#5
 ]
}
emitters {
 Glows {
  emitterUpdateScript [
   &EmitterState(
    Inactive Response=::NewEnumerator1
    Life Cycle Mode=::NewEnumerator1
    Loop Behavior=::NewEnumerator1
    Loop Duration Mode=::NewEnumerator1
    UseLoopDelay=false
   )
   &SpawnBurst_Instantaneous(Spawn Count=6, Use Spawn Probability=false, UseLoopCountLimit=false)
  ]
  particleSpawnScript [
   &PartitionParticles(Number Of Partitions=6)
   &InitializeParticle(Lifetime=2, Sprite Size Mode=::NewEnumerator3, Uniform Sprite Size=20, UsePositionOffset=false, Write Material Random=true)
   &CameraOffset(Camera Offset Amount=3)
  ]
  particleUpdateScript [
   &ParticleState(Kill Particles When Lifetime Has Elapsed=false, Let Infinitely Lived Particles Die When Emitter Deactivates=true)
   &ScaleSpriteSize(
    Scale Sprite Size Mode=::NewEnumerator0
    Uniform Scale Factor=&MergeHelper_FloatSwitch(0=1.3, 1=1, 2=1.3, 3=1, 4=1.3, 5=1.3, Index=$Particles.PartitionParticles.Partition, Num=::NewEnumerator4)
   ) {module=ScaleSpriteSize001}
   &ScaleColor(
    Scale Alpha=&MergeHelper_FloatSwitch(0=1, 1=0.7, 2=1, 3=0.7, 4=0.6, 5=0.6, Index=$Particles.PartitionParticles.Partition, Num=::NewEnumerator4)
    ScaleA=true
    ScaleRGB=true
   ) {module=ScaleColor002}
   &SkeletalMeshLocation(
    Bone / Socket Index=&SelectIntFromArray(Direct Array Index=$Particles.PartitionParticles.Partition, Int Selection Array=DataInterface() {propertyValues {IntData [1,1,5,5,10,12]}})
    Bone Sampling Mode=::NewEnumerator4
    MaskSpawnGroup=false
    Orientation Sampling=::NewEnumerator1
    Sampled Position Offset=&MergeHelper_VectorSwitch(
     0=&FN_FrontEndVector3Modifier(Gameplay=<0 1 -2>, Front End=<-2 0 -2>)
     1=&FN_FrontEndVector3Modifier(Gameplay=<5 1 2>, Front End=<3 0 2.5>)
     2=&FN_FrontEndVector3Modifier(Gameplay=<0 0 2>, Front End=<0 -1 2>)
     3=&FN_FrontEndVector3Modifier(Gameplay=<-5 -1 -2>, Front End=<-3 0 -2.5>)
     4=<-2 0 -1>
     5=<2 0 1>
     Index=$Particles.PartitionParticles.Partition
     Num=::NewEnumerator4
    )
    Skeletal Mesh=$System.Skeletal Mesh Reader
   )
   &ScaleSpriteSize(
    Uniform Curve Index=$Emitter.Age
    Uniform Curve Sprite Scale=DataInterface() {propertyValues {Curve {defaultValue=1, keys [{tangentMode=RCTM_User, time=0, value=0}, {tangentMode=RCTM_User, time=0.1, value=1}]}}}
   )
   &ScaleColor(
    Scale Mode=1
    Scale RGBA=&MakeVector4_FromLinearColor(LinearColor=&MultiplyLinearColor(
      Linear Color=$System.Color Array
      Scale Factor=&MakeVector4_FromLinearColor(LinearColor=/Niagara/DynamicInputs/TypeConversions/MakeLinearColorFromFloat(Value:NiagaraFloat=&Multiply_Float(A=$System.System_FadeIn, B=&FN_FrontEndFloatModifier(Gameplay=0.6, Front End=0.4))))
     ))
    ScaleRGBA=true
   )
  ]
  rendererClasses [&NiagaraSpriteRendererProperties]
  renderers [{rendererClass=&NiagaraSpriteRendererProperties}]
 }
 Lights {
  emitterUpdateScript [
   &EmitterState(
    Inactive Response=::NewEnumerator1
    Life Cycle Mode=::NewEnumerator1
    Loop Behavior=::NewEnumerator1
    Loop Duration Mode=::NewEnumerator1
    UseLoopDelay=false
   )
   &SpawnBurst_Instantaneous(Spawn Count=4, Use Spawn Probability=false, UseLoopCountLimit=false)
  ]
  particleSpawnScript [
   &PartitionParticles(Number Of Partitions=4)
   &InitializeParticle(Lifetime=2, Sprite Size Mode=::NewEnumerator3, Uniform Sprite Size=20, UsePositionOffset=false, Write Material Random=true)
   &CameraOffset(Camera Offset Amount=3)
  ]
  particleUpdateScript [
   &ParticleState(Kill Particles When Lifetime Has Elapsed=false, Let Infinitely Lived Particles Die When Emitter Deactivates=true)
   &SkeletalMeshLocation(
    Bone / Socket Index=&SelectIntFromArray(Direct Array Index=$Particles.PartitionParticles.Partition, Int Selection Array=DataInterface() {propertyValues {IntData [1,5,10,12]}})
    Bone Sampling Mode=::NewEnumerator4
    MaskSpawnGroup=false
    Orientation Sampling=::NewEnumerator1
    Sampled Position Offset=&MergeHelper_VectorSwitch(0=<2 -0.5 -1>, 1=<0 -0.5 -1>, 2=<2 0 -2>, 3=<-2 0 -2>, Index=$Particles.PartitionParticles.Partition, Num=::NewEnumerator2)
    Sampled Position Offset Coordinate Space:ENiagaraSkelMeshTransforms=::NewEnumerator2
    Skeletal Mesh=$System.Skeletal Mesh Reader
   )
   &ScaleSpriteSize(
    Uniform Curve Index=$Emitter.Age
    Uniform Curve Sprite Scale=DataInterface() {propertyValues {Curve {defaultValue=1, keys [{tangentMode=RCTM_User, time=0, value=0}, {tangentMode=RCTM_User, time=0.1, value=1}]}}}
   )
   &ScaleColor(
    Scale Alpha=&Multiply_Float(
     A=$System.System_FadeIn
     B=&Multiply_Float(
      A=&FN_FrontEndFloatModifier(Gameplay=0, Front End=0)
      B=&MakeFloatFromLinearColor(LinearColor=$System.Color Array, Channel:ENiagaraFloatFromLinearColorOptions=::NewEnumerator3)
     )
    )
    Scale RGB=&MakeVectorFromLinearColorRGB(Color=$System.Color Array_Lights)
    ScaleA=true
    ScaleRGB=true
   )
   &FN_Light_Attributes(
    Color=&MultiplyLinearColor(
     Linear Color=$System.Color Array_Lights
     Scale Factor=&MakeVector4_FromFloat(Float=&Multiply_Float(A=$System.System_FadeIn, B=&FN_FrontEndFloatModifier(Gameplay=0.4, Front End=0.3)))
    )
    Create Color Attribute=true
    Create Exponent Attribute=true
    Create Radius Attribute=true
    Exponent=1
    Radius=10
    Scale Color=5
   )
  ]
  rendererClasses [&NiagaraLightRendererProperties]
  renderers [{rendererClass=&NiagaraLightRendererProperties, rendererIndex=0}]
 }
 Trails {
  emitterUpdateScript [
   &EmitterState(
    Inactive Response=::NewEnumerator1
    Life Cycle Mode=::NewEnumerator1
    Loop Behavior=::NewEnumerator1
    Loop Duration Mode=::NewEnumerator1
    UseLoopDelay=false
   )
   &SpawnBurst_Instantaneous(Spawn Count=&Multiply_Int(A=4, B=4), Use Spawn Probability=false, UseLoopCountLimit=false)
  ]
  particleSpawnScript [
   &PartitionParticles(
    Normalized Index Includes One=true
    Number Of Partitions=4
    Particles Per Partition=4
    Partition Mode:ENiagara_PartitionMode_v2=::NewEnumerator1
    Write Partition to Ribbon ID=::NewEnumerator0
   )
   &InitializeParticle(
    Lifetime=2
    Ribbon Width=7
    Ribbon Width Mode=::NewEnumerator1
    Sprite Size Mode=::NewEnumerator3
    Uniform Sprite Size=35
    UsePositionOffset=false
    Write Material Random=true
   )
   &CameraOffset(Camera Offset Amount=3)
  ]
  particleUpdateScript [
   &ParticleState(Kill Particles When Lifetime Has Elapsed=false, Let Infinitely Lived Particles Die When Emitter Deactivates=true)
   /Game/Effects/Niagara/Modules/Curves/FN_ListCurve_SkeletalMeshSampler(
    Apply Orientation=true
    Apply Position=true
    Curve Index:NiagaraFloat=$Particles.PartitionParticles.Normalized Index In Partition
    Index_01:NiagaraInt32=&MergeHelper_IntSwitch(0=0, 1=4, 2=8, 3=8, 4=?, 5=?, Index=$Particles.PartitionParticles.Partition, Num=::NewEnumerator2)
    Index_02:NiagaraInt32=&MergeHelper_IntSwitch(0=1, 1=5, 2=9, 3=9, 4=?, 5=?, Index=$Particles.PartitionParticles.Partition, Num=::NewEnumerator2)
    Index_03:NiagaraInt32=&MergeHelper_IntSwitch(0=2, 1=6, 2=10, 3=12, 4=?, 5=?, Index=$Particles.PartitionParticles.Partition, Num=::NewEnumerator2)
    Index_04:NiagaraInt32=&MergeHelper_IntSwitch(0=3, 1=7, 2=11, 3=13, 4=?, 5=?, Index=$Particles.PartitionParticles.Partition, Num=::NewEnumerator2)
    Offset Coordinate Space=3
    Offset_01:Vector3f=&MergeHelper_VectorSwitch(0=<0 0 0>, 1=<0 0 0>, 2=<0 0 0>, 3=<0 0 0>, Index=$Particles.PartitionParticles.Partition, Num=::NewEnumerator2)
    Offset_02:Vector3f=&MergeHelper_VectorSwitch(0=<0 0 0>, 1=<0 0 0>, 2=<0 0 0>, 3=<0 0 0>, Index=$Particles.PartitionParticles.Partition, Num=::NewEnumerator2)
    Offset_03:Vector3f=&MergeHelper_VectorSwitch(0=<0 0 0>, 1=<0 0 0>, 2=<0 0 -2>, 3=<0 -1 2>, Index=$Particles.PartitionParticles.Partition, Num=::NewEnumerator2)
    Offset_04:Vector3f=&MergeHelper_VectorSwitch(0=<0 0 0>, 1=<0 0 0>, 2=<0 0 -4>, 3=<0 0 4>, Index=$Particles.PartitionParticles.Partition, Num=::NewEnumerator2)
    Skeletal Mesh:NiagaraDataInterfaceSkeletalMesh=$System.Skeletal Mesh Reader
   )
   &ScaleSpriteSize(
    Uniform Curve Index=$Emitter.Age
    Uniform Curve Sprite Scale=DataInterface() {propertyValues {Curve {defaultValue=1, keys [{tangentMode=RCTM_User, time=0, value=0}, {tangentMode=RCTM_User, time=0.1, value=1}]}}}
   )
   &ScaleColor(
    Scale Mode=1
    Scale RGBA=&MakeVector4_FromLinearColor(LinearColor=&MultiplyLinearColor(
      Linear Color=$System.Color Array_Energy
      Scale Factor=&MakeVector4_FromLinearColor(LinearColor=&MakeLinearColorFromFloat(Value:NiagaraFloat=&Multiply_Float(A=$System.System_FadeIn, B=&FN_FrontEndFloatModifier(Gameplay=5, Front End=5))))
     ))
    ScaleRGBA=true
   )
  ]
  rendererClasses [&NiagaraRibbonRendererProperties]
  renderers [{rendererClass=&NiagaraRibbonRendererProperties}]
 }
}
summary {
 systemName=NS_Character_PixieParts_Idle
 emitters [
  {emitterName=Trails, rendererClasses [&NiagaraRibbonRendererProperties]}
  {emitterName=Lights, rendererClasses [&NiagaraLightRendererProperties]}
  {emitterName=Glows, rendererClasses [&NiagaraSpriteRendererProperties]}
 ]
 userVariables [
  User.bIsFrontEnd {type=NiagaraBool}
  User.bIsFrontEndPreview {type=NiagaraBool}
  User.Variant {type=NiagaraFloat, defaultValue=0, description=""}
 ]
}
