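# Material function for the cockpit material of the PixieParts Lava player body cosmetic, layering fabric, plastic and metal looks onto incoming material attributes.
# Each look runs MF_PixieParts_SimpleMaterial with its own tint, albedo, roughness and specular remaps, plus a metallic offset.
# Three chained attribute blends use channel-selectable masks and multipliers, keeping vertex attributes from the first input.
# The blended result passes through TechArt cloth fuzz, then metal LUT functions before output.
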MaterialFunction /BRCosmetics/Characters/Player/Male/Medium/Bodies/M_MED_PixieParts_Lava/Materials/MaterialFunctions/MF_PixieParts_CockpitMaterial
graph {nodes [
  n1 = NamedRerouteUsage()
  n2 = NamedRerouteUsage()
  n3 = Multiply(A=NamedRerouteUsage(), B=NamedRerouteUsage())
  n4 = NamedRerouteUsage()
  n5 = NamedRerouteUsage()
  n6 = BlendMaterialAttributes(
   A=BlendMaterialAttributes(A=BlendMaterialAttributes(A=@n1, B=@n2, Alpha=@n3) {vertexAttributeBlendType=UseA}, B=NamedRerouteUsage(), Alpha=Multiply(A=@n4, B=@n5)) {vertexAttributeBlendType=UseA}
   B=NamedRerouteUsage()
   Alpha=Multiply(A=NamedRerouteUsage(), B=NamedRerouteUsage())
  ) {vertexAttributeBlendType=UseA}
  n7 = NamedRerouteUsage()
  n8 = NamedRerouteUsage()
  n9 = NamedRerouteUsage()
  FunctionOutput(None=MaterialFunctionCall!(
    Alpha=NamedRerouteUsage()
    Materials=MaterialFunctionCall!(Material=@n6, Ao/Gmap/Skin=@n7, Fresnel=@n8, Alpha=@n9) {materialFunction=/Game/Global/MaterialFunctions/TechArt/MF_ClothFuzz_TechArt}.Result
   ) {materialFunction=&/MF_MetalLUT_TechArt}.Result) {outputName="Material Out (MA)"}
  NamedRerouteDeclaration(None=FunctionInput() {inputName=Material In, inputType=FunctionInput_MaterialAttributes}) {name="Material In (MA)", nodeColor=<rgba 0.4531 1 0 1>}
  NamedRerouteDeclaration(None=FunctionInput() {inputName=TechArt Mask, sortPriority=10}) {name="FX Mask In (V3)", nodeColor=<rgba 0.5703 0 1 1>}
  n10 = VectorParameter() {
   group=PixieParts | Materials | Metal
   parameterName=Metal Mat - Albedo Remap
   sortPriority=10
   channelNames {a=Gain, b=Exp, g=S-Curve Exp, r=Desat}
   defaultValue=<rgba 1 1 1 1>
  }
  NamedRerouteDeclaration(None=@n10.RGBA) {name="Metal Mat - Albedo Remap (V4)", nodeColor=<rgba 1 0.7468 0.1679 1>}
  n11 = VectorParameter() {
   group=PixieParts | Materials | Metal
   parameterName=Metal Mat - Roughness Remap
   sortPriority=50
   channelNames {a=n/a, b=Mult, g=S-Curve Exp, r=Offset}
   defaultValue=<rgba 0 1 1 0>
  }
  NamedRerouteDeclaration(None=@n11.RGB) {name="Metal Mat - Roughness Remap (V3)", nodeColor=<rgba 1 0.7468 0.1679 1>}
  n12 = VectorParameter() {
   group=PixieParts | Materials | Metal
   parameterName=Metal Mat - Specular Remap
   sortPriority=40
   channelNames {a=Metallic Offset, b=Mult, g=S-Curve Exp, r=Offset}
   defaultValue=<rgba 0 1 1 0>
  }
  NamedRerouteDeclaration(None=@n12.RGB) {name="Metal Mat - Specular Remap (V3)", nodeColor=<rgba 1 0.7468 0.1679 1>}
  NamedRerouteDeclaration(None=@n12.Metallic Offset) {name="Metal Mat - Metallic Offset (S)", nodeColor=<rgba 1 0.7468 0.1679 1>}
  n13 = VectorParameter() {
   group=PixieParts | Materials | Metal
   parameterName=Metal Mat - Color
   sortPriority=20
   channelNames {a=n/a, b="", g="", r=""}
   defaultValue=<rgba 0.18 0.18 0.18 1>
  }
  NamedRerouteDeclaration(None=@n13.RGB) {name="Metal Mat - Color (V3)", nodeColor=<rgba 1 0.7468 0.1679 1>}
  n14 = VectorParameter() {
   group=PixieParts | Materials | Fabric
   parameterName=Fabric Mat - Roughness Remap
   sortPriority=50
   channelNames {a=n/a, b=Mult, g=S-Curve Exp, r=Offset}
   defaultValue=<rgba 0 1 1 0>
  }
  NamedRerouteDeclaration(None=@n14.RGB) {name="Fabric Mat - Roughness Remap (V3)", nodeColor=<rgba 0.1679 0.4863 1 1>}
  n15 = VectorParameter() {
   group=PixieParts | Materials | Fabric
   parameterName=Fabric Mat - Specular Remap
   sortPriority=40
   channelNames {a=Metallic Offset, b=Mult, g=S-Curve Exp, r=Offset}
   defaultValue=<rgba 0 1 1 0>
  }
  NamedRerouteDeclaration(None=@n15.RGB) {name="Fabric Mat - Specular Remap (V3)", nodeColor=<rgba 0.1679 0.4863 1 1>}
  n16 = VectorParameter() {
   group=PixieParts | Materials | Fabric
   parameterName=Fabric Mat - Color
   sortPriority=20
   channelNames {a=n/a, b="", g="", r=""}
   defaultValue=<rgba 0.18 0.18 0.18 1>
  }
  NamedRerouteDeclaration(None=@n16.RGB) {name="Fabric Mat - Color (V3)", nodeColor=<rgba 0.1679 0.4863 1 1>}
  n17 = VectorParameter() {
   group=PixieParts | Materials | Fabric
   parameterName=Fabric Mat - Albedo Remap
   sortPriority=10
   channelNames {a=Gain, b=Exp, g=S-Curve Exp, r=Desat}
   defaultValue=<rgba 1 1 1 1>
  }
  NamedRerouteDeclaration(None=@n17.RGBA) {name="Fabric Mat - Albedo Remap (V4)", nodeColor=<rgba 0.1679 0.4863 1 1>}
  NamedRerouteDeclaration(None=@n15.Metallic Offset) {name="Fabric Mat - Metallic Offset (S)", nodeColor=<rgba 0.1679 0.4863 1 1>}
  n18 = NamedRerouteUsage()
  NamedRerouteDeclaration(None=GetMaterialAttributes(None=@n18) {attributeGetTypes [#1]}.Metallic) {name="Metallic In (S)", nodeColor=<rgba 0 1 0.1797 1>}
  n19 = FunctionInput() {bUsePreviewValueAsDefault=true, inputName=Fresnel In, inputType=FunctionInput_Scalar, sortPriority=30, previewValue=<0 1 0 1>}
  NamedRerouteDeclaration(None=@n19) {name="Fresnel In (S)", nodeColor=<rgba 0 1 0.0156 1>}
  n20 = FunctionInput() {bUsePreviewValueAsDefault=true, inputName=AO/GMap/Skin, sortPriority=20, previewValue=<1 1 0 1>}
  NamedRerouteDeclaration(None=@n20) {name="Mask In (V3)", nodeColor=<rgba 1 0 0.3047 1>}
  NamedRerouteDeclaration(None=ChannelMaskParameter(None=NamedRerouteUsage()) {group=PixieParts | Globals, parameterName=FabricMaskChannel, sortPriority=-1000}.R) {name="Fabric Material Mask (S)", nodeColor=<rgba 0.1679 0.4863 1 1>}
  NamedRerouteDeclaration(None=ScalarParameter() {defaultValue=1, group=PixieParts | Globals, parameterName=Fabric Mask Mult, sliderMax=1, sortPriority=-990}) {name="Fabric Mat - Mask Mult (S)", nodeColor=<rgba 0.1679 0.4863 1 1>}
  NamedRerouteDeclaration(None=ScalarParameter() {defaultValue=1, group=PixieParts | Globals, parameterName=Plastic Mask Mult, sliderMax=1, sortPriority=-940}) {name="Plastic Mat - Mask Mult (S)", nodeColor=<rgba 0.4139 1 0.1679 1>}
  NamedRerouteDeclaration(None=ScalarParameter() {defaultValue=1, group=PixieParts | Globals, parameterName=Metal Mask Mult, sliderMax=1, sortPriority=-890}) {name="Metal Mat - Mask Mult (S)", nodeColor=<rgba 1 0.7468 0.1679 1>}
  n21 = MaterialFunctionCall!(
   Material In=@n18
   Albedo Remaps=NamedRerouteUsage()
   Tint=NamedRerouteUsage()
   Metallic=Add(A=NamedRerouteUsage(), B=NamedRerouteUsage())
   Specular Remaps=NamedRerouteUsage()
   Roughness Remaps=NamedRerouteUsage()
  ) {materialFunction=&/MF_PixieParts_SimpleMaterial}
  NamedRerouteDeclaration(None=@n21."Material Out (MA)") {name="Fabric Material  (MA)", nodeColor=<rgba 0.1679 0.4863 1 1>}
  n22 = MaterialFunctionCall(
   Material In=NamedRerouteUsage()
   Albedo Remaps=NamedRerouteUsage()
   Tint=NamedRerouteUsage()
   Metallic=Add(A=NamedRerouteUsage(), B=NamedRerouteUsage())
   Specular Remaps=NamedRerouteUsage()
   Roughness Remaps=NamedRerouteUsage()
  ) {materialFunction=&MF_PixieParts_SimpleMaterial}
  NamedRerouteDeclaration(None=@n22."Material Out (MA)") {name="Plastic Material (MA)", nodeColor=<rgba 0.4139 1 0.1679 1>}
  n23 = MaterialFunctionCall(
   Material In=NamedRerouteUsage()
   Albedo Remaps=NamedRerouteUsage()
   Tint=NamedRerouteUsage()
   Metallic=Add(A=NamedRerouteUsage(), B=NamedRerouteUsage())
   Specular Remaps=NamedRerouteUsage()
   Roughness Remaps=NamedRerouteUsage()
  ) {materialFunction=&MF_PixieParts_SimpleMaterial}
  NamedRerouteDeclaration(None=@n23."Material Out (MA)") {name="Metal Material (MA)", nodeColor=<rgba 1 0.7468 0.1679 1>}
  n24 = VectorParameter() {
   group=PixieParts | Materials | Plastic
   parameterName=Plastic Mat - Albedo Remap
   sortPriority=10
   channelNames {a=Gain, b=Exp, g=S-Curve Exp, r=Desat}
   defaultValue=<rgba 1 1 1 1>
  }
  NamedRerouteDeclaration(None=@n24.RGBA) {name="Plastic Mat - Albedo Remap (V4)", nodeColor=<rgba 0.4139 1 0.1679 1>}
  n25 = VectorParameter() {
   group=PixieParts | Materials | Plastic
   parameterName=Plastic Mat - Roughness Remap
   sortPriority=50
   channelNames {a=n/a, b=Mult, g=S-Curve Exp, r=Offset}
   defaultValue=<rgba 0 1 1 0>
  }
  NamedRerouteDeclaration(None=@n25.RGB) {name="Plastic Mat - Roughness Remap (V3)", nodeColor=<rgba 0.4139 1 0.1679 1>}
  n26 = VectorParameter() {
   group=PixieParts | Materials | Plastic
   parameterName=Plastic Mat - Specular Remap
   sortPriority=40
   channelNames {a=Metallic Offset, b=Mult, g=S-Curve Exp, r=Offset}
   defaultValue=<rgba 0 1 1 0>
  }
  NamedRerouteDeclaration(None=@n26.RGB) {name="Plastic Mat - Specular Remap (V3)", nodeColor=<rgba 0.4139 1 0.1679 1>}
  NamedRerouteDeclaration(None=@n26.Metallic Offset) {name="Plastic Mat - Metallic Offset (S)", nodeColor=<rgba 0.4139 1 0.1679 1>}
  n27 = VectorParameter() {
   group=PixieParts | Materials | Plastic
   parameterName=Plastic Mat - Color
   sortPriority=20
   channelNames {a=n/a, b="", g="", r=""}
   defaultValue=<rgba 0.18 0.18 0.18 1>
  }
  NamedRerouteDeclaration(None=@n27.RGB) {name="Plastic Mat - Color (V3)", nodeColor=<rgba 0.4139 1 0.1679 1>}
  n28 = ChannelMaskParameter(None=NamedRerouteUsage()) {group=PixieParts | Globals, maskChannel=Blue, parameterName=PlasticMaskChannel, sortPriority=-950, defaultValue=<rgba 0 0 1 0>}
  NamedRerouteDeclaration(None=@n28.R) {name="Plastic Material Mask (S)", nodeColor=<rgba 0.4139 1 0.1679 1>}
  n29 = ChannelMaskParameter(None=NamedRerouteUsage()) {group=PixieParts | Globals, maskChannel=Green, parameterName=MetalMaskChannel, sortPriority=-900, defaultValue=<rgba 0 1 0 0>}
  NamedRerouteDeclaration(None=@n29.R) {name="Metal Material Mask (S)", nodeColor=<rgba 1 0.7468 0.1679 1>}
 ]}
