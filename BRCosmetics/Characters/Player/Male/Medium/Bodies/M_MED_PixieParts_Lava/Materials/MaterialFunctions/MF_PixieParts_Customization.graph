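# Material function that assembles the customizable layered material for the PixieParts player body cosmetic.
# Base, stone and mask inputs feed sub-functions for material variants by slot ID, special materials (gold, galaxy, crystal, pearl), decals and moss.
# Mask-driven blend functions combine the layers, including energy emission and a visor override.
# Quality switches limit thin film, clear coat, metal LUT and visor to higher tiers, and BladeFX blends into emissive.
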MaterialFunction /BRCosmetics/Characters/Player/Male/Medium/Bodies/M_MED_PixieParts_Lava/Materials/MaterialFunctions/MF_PixieParts_Customization
graph {expressions {
  MaterialExpressionBlendMaterialAttributes_0 {
   class=MaterialExpressionBlendMaterialAttributes
   inputs [
    A {from=MaterialExpressionMaterialFunctionCall_11, output="Material Out (MA)"}
    B {from=MaterialExpressionNamedRerouteUsage_9, output=""}
    Alpha {from=MaterialExpressionMultiply_0, output=""}
   ]
   properties {}
  }
  MaterialExpressionBlendMaterialAttributes_2 {
   class=MaterialExpressionBlendMaterialAttributes
   inputs [
    A {from=MaterialExpressionMaterialFunctionCall_32, output=Result}
    B {from=MaterialExpressionNamedRerouteUsage_346, output=""}
    Alpha {from=MaterialExpressionNamedRerouteUsage_347, output=""}
   ]
   properties {vertexAttributeBlendType=UseA}
  }
  MaterialExpressionChannelMaskParameter_1 {
   class=MaterialExpressionChannelMaskParameter
   inputs [None {from=MaterialExpressionFunctionInput_5, output=""}]
   properties {group=PixieParts | Materials | Moss, maskChannel=Green, parameterName=Moss Mask Channel, sortPriority=10, defaultValue=<rgba 0 1 0 0>}
  }
  MaterialExpressionChannelMaskParameter_2 {
   class=MaterialExpressionChannelMaskParameter
   inputs [None {from=MaterialExpressionFunctionInput_5, output=""}]
   properties {group=PixieParts | Materials | Moss, parameterName=Visor Override Mask Channel, sortPriority=10}
  }
  MaterialExpressionClamp_0 {
   class=MaterialExpressionClamp
   inputs [
    None {from=MaterialExpressionScalarParameter_26, output=""}
    Min {from="", output=""}
    Max {from="", output=""}
   ]
   properties {maxDefault=2}
  }
  MaterialExpressionConstant_1 {class=MaterialExpressionConstant, properties {r=1}}
  MaterialExpressionConstant_2 {class=MaterialExpressionConstant}
  MaterialExpressionConstant_3 {class=MaterialExpressionConstant, properties {r=-1}}
  MaterialExpressionConstant_4 {class=MaterialExpressionConstant, properties {r=1}}
  MaterialExpressionConstant_5 {class=MaterialExpressionConstant, properties {r=0.5}}
  MaterialExpressionConvert_0 {
   class=MaterialExpressionConvert
   inputs [None {from=MaterialExpressionNamedRerouteUsage_33}]
   properties {
    nodeName=Break Float3
    convertInputs [{
      type=Vector3
      expressionInput {
       expression=&self:MaterialExpressionNamedRerouteUsage_33
       inputName=None
       mask=0
       maskA=0
       maskB=0
       maskG=0
       maskR=0
       outputIndex=0
      }
     }]
    convertOutputs [{type=Scalar}, {type=Scalar}, {type=Scalar}]
   }
  }
  MaterialExpressionConvert_1 {
   class=MaterialExpressionConvert
   inputs [None {from=MaterialExpressionNamedRerouteUsage_34}]
   properties {
    nodeName=Break Float3
    convertInputs [{
      type=Vector3
      expressionInput {
       expression=&self:MaterialExpressionNamedRerouteUsage_34
       inputName=None
       mask=0
       maskA=0
       maskB=0
       maskG=0
       maskR=0
       outputIndex=0
      }
     }]
    convertOutputs [{type=Scalar}, {type=Scalar}, {type=Scalar}]
   }
  }
  MaterialExpressionConvert_17 {
   class=MaterialExpressionConvert
   inputs [None {from=MaterialExpressionMultiply_14}]
   properties {
    nodeName=Break Float3
    convertInputs [{
      type=Vector3
      expressionInput {
       expression=&self:MaterialExpressionMultiply_14
       inputName=None
       mask=0
       maskA=0
       maskB=0
       maskG=0
       maskR=0
       outputIndex=0
      }
     }]
    convertOutputs [{type=Scalar}, {type=Scalar}, {type=Scalar}]
   }
  }
  MaterialExpressionConvert_19 {
   class=MaterialExpressionConvert
   inputs [None {from=MaterialExpressionRound_7}]
   properties {
    nodeName=Break Float3
    convertInputs [{
      type=Vector3
      expressionInput {
       expression=&self:MaterialExpressionRound_7
       inputName=None
       mask=0
       maskA=0
       maskB=0
       maskG=0
       maskR=0
       outputIndex=0
      }
     }]
    convertOutputs [{type=Scalar}, {type=Scalar}, {type=Scalar}]
   }
  }
  MaterialExpressionConvert_2 {class=MaterialExpressionConvert, inputs [], properties {convertOutputs [{type=Vector3, defaultValue=<rgba 1 0 0 1>}]}}
  MaterialExpressionConvert_20 {
   class=MaterialExpressionConvert
   inputs [
    None {from=MaterialExpressionScalarParameter_4}
    None {from=MaterialExpressionScalarParameter_7}
    None {from=MaterialExpressionScalarParameter_17}
   ]
   properties {
    nodeName=Make Vector3
    convertInputs [
     {
      type=Scalar
      expressionInput {
       expression=&self:MaterialExpressionScalarParameter_4
       inputName=None
       mask=0
       maskA=0
       maskB=0
       maskG=0
       maskR=0
       outputIndex=0
      }
     }
     {
      type=Scalar
      expressionInput {
       expression=&self:MaterialExpressionScalarParameter_7
       inputName=None
       mask=0
       maskA=0
       maskB=0
       maskG=0
       maskR=0
       outputIndex=0
      }
     }
     {
      type=Scalar
      expressionInput {
       expression=&self:MaterialExpressionScalarParameter_17
       inputName=None
       mask=0
       maskA=0
       maskB=0
       maskG=0
       maskR=0
       outputIndex=0
      }
     }
    ]
    convertOutputs [{type=Vector3}]
   }
  }
  MaterialExpressionConvert_21 {
   class=MaterialExpressionConvert
   inputs [None {from=MaterialExpressionNamedRerouteDeclaration_54}]
   properties {
    convertInputs [{
      type=Scalar
      expressionInput {
       expression=&self:MaterialExpressionNamedRerouteDeclaration_54
       inputName=None
       mask=0
       maskA=0
       maskB=0
       maskG=0
       maskR=0
       outputIndex=0
      }
     }]
    convertOutputs [{type=Vector3}, {type=Vector2}]
   }
  }
  MaterialExpressionConvert_22 {
   class=MaterialExpressionConvert
   inputs [None {from=MaterialExpressionNamedRerouteUsage_104}]
   properties {
    nodeName=Break Float3
    convertInputs [{
      type=Vector3
      expressionInput {
       expression=&self:MaterialExpressionNamedRerouteUsage_104
       inputName=None
       mask=0
       maskA=0
       maskB=0
       maskG=0
       maskR=0
       outputIndex=0
      }
     }]
    convertOutputs [{type=Scalar}, {type=Scalar}, {type=Scalar}]
   }
  }
  MaterialExpressionConvert_23 {
   class=MaterialExpressionConvert
   inputs [None {from=MaterialExpressionFunctionInput_6}]
   properties {
    convertInputs [{
      type=Vector3
      expressionInput {
       expression=&self:MaterialExpressionFunctionInput_6
       inputName=None
       mask=0
       maskA=0
       maskB=0
       maskG=0
       maskR=0
       outputIndex=0
      }
     }]
    convertOutputs [{type=Scalar}, {type=Scalar}]
   }
  }
  MaterialExpressionConvert_24 {
   class=MaterialExpressionConvert
   inputs [
    None {from=MaterialExpressionScalarParameter_18}
    None {from=MaterialExpressionScalarParameter_20}
   ]
   properties {
    nodeName=Make Vector2
    convertInputs [
     {
      type=Scalar
      expressionInput {
       expression=&self:MaterialExpressionScalarParameter_18
       inputName=None
       mask=0
       maskA=0
       maskB=0
       maskG=0
       maskR=0
       outputIndex=0
      }
     }
     {
      type=Scalar
      expressionInput {
       expression=&self:MaterialExpressionScalarParameter_20
       inputName=None
       mask=0
       maskA=0
       maskB=0
       maskG=0
       maskR=0
       outputIndex=0
      }
     }
    ]
    convertOutputs [{type=Vector2}]
   }
  }
  MaterialExpressionConvert_25 {
   class=MaterialExpressionConvert
   inputs [None {from=MaterialExpressionRound_8}]
   properties {
    nodeName=Break Float2
    convertInputs [{
      type=Vector2
      expressionInput {
       expression=&self:MaterialExpressionRound_8
       inputName=None
       mask=0
       maskA=0
       maskB=0
       maskG=0
       maskR=0
       outputIndex=0
      }
     }]
    convertOutputs [{type=Scalar}, {type=Scalar}]
   }
  }
  MaterialExpressionConvert_3 {class=MaterialExpressionConvert, inputs [], properties {convertOutputs [{type=Vector3, defaultValue=<rgba 0 1 0 1>}]}}
  MaterialExpressionConvert_4 {class=MaterialExpressionConvert, inputs [], properties {convertOutputs [{type=Vector3, defaultValue=<rgba 1 0 0 1>}]}}
  MaterialExpressionConvert_5 {class=MaterialExpressionConvert, inputs [], properties {convertOutputs [{type=Vector3, defaultValue=<rgba 0 0 1 1>}]}}
  MaterialExpressionConvert_6 {
   class=MaterialExpressionConvert
   inputs [None {from=MaterialExpressionSwitch_2}]
   properties {
    nodeName=Break Float3
    convertInputs [{
      type=Vector3
      expressionInput {
       expression=&self:MaterialExpressionSwitch_2
       inputName=None
       mask=0
       maskA=0
       maskB=0
       maskG=0
       maskR=0
       outputIndex=0
      }
     }]
    convertOutputs [{type=Scalar}, {type=Scalar}, {type=Scalar}]
   }
  }
  MaterialExpressionConvert_7 {class=MaterialExpressionConvert, inputs [], properties {convertOutputs [{type=Vector3, defaultValue=<rgba 0 0 1 1>}]}}
  MaterialExpressionFunctionInput_0 {
   class=MaterialExpressionFunctionInput
   inputs [Preview {from="", output=""}]
   properties {inputName=Base Material In, inputType=FunctionInput_MaterialAttributes}
  }
  MaterialExpressionFunctionInput_1 {class=MaterialExpressionFunctionInput, inputs [Preview {from="", output=""}], properties {inputName=Material Masks A, sortPriority=20}}
  MaterialExpressionFunctionInput_3 {class=MaterialExpressionFunctionInput, inputs [Preview {from="", output=""}], properties {inputName=Material Masks B, sortPriority=30}}
  MaterialExpressionFunctionInput_5 {
   class=MaterialExpressionFunctionInput
   inputs [Preview {from="", output=""}]
   properties {bUsePreviewValueAsDefault=true, inputName=Decal Masks, sortPriority=40}
  }
  MaterialExpressionFunctionInput_6 {
   class=MaterialExpressionFunctionInput
   inputs [Preview {from="", output=""}]
   properties {bUsePreviewValueAsDefault=true, inputName=AO/GMap/Skin, sortPriority=50, previewValue=<1 1 0 1>}
  }
  MaterialExpressionFunctionInput_7 {
   class=MaterialExpressionFunctionInput
   inputs [Preview {from="", output=""}]
   properties {bUsePreviewValueAsDefault=true, inputName=Fresnel In, inputType=FunctionInput_Scalar, sortPriority=60, previewValue=<0 1 0 1>}
  }
  MaterialExpressionFunctionInput_8 {
   class=MaterialExpressionFunctionInput
   inputs [Preview {from="", output=""}]
   properties {inputName=Stone Material In, inputType=FunctionInput_MaterialAttributes, sortPriority=10}
  }
  MaterialExpressionFunctionOutput_3 {
   class=MaterialExpressionFunctionOutput
   inputs [None {from=MaterialExpressionSetMaterialAttributes_4, output=""}]
   properties {outputName="Material Out (MA)"}
  }
  MaterialExpressionGetMaterialAttributes_7 {
   class=MaterialExpressionGetMaterialAttributes
   inputs [None {from=MaterialExpressionNamedRerouteUsage_354, output=""}]
   properties {attributeGetTypes [#1]}
  }
  MaterialExpressionLinearInterpolate_1 {
   class=MaterialExpressionLinearInterpolate
   inputs [
    A {from=""}
    B {from=MaterialExpressionNamedRerouteUsage_112}
    Alpha {from=MaterialExpressionNamedRerouteUsage_106}
   ]
  }
  MaterialExpressionLinearInterpolate_2 {
   class=MaterialExpressionLinearInterpolate
   inputs [
    A {from=MaterialExpressionLinearInterpolate_1}
    B {from=MaterialExpressionNamedRerouteUsage_113}
    Alpha {from=MaterialExpressionNamedRerouteUsage_109}
   ]
  }
  MaterialExpressionLinearInterpolate_3 {
   class=MaterialExpressionLinearInterpolate
   inputs [
    A {from=MaterialExpressionLinearInterpolate_2}
    B {from=MaterialExpressionNamedRerouteUsage_114}
    Alpha {from=MaterialExpressionNamedRerouteUsage_110}
   ]
  }
  MaterialExpressionLinearInterpolate_4 {
   class=MaterialExpressionLinearInterpolate
   inputs [
    A {from=MaterialExpressionLinearInterpolate_3}
    B {from=MaterialExpressionNamedRerouteUsage_115}
    Alpha {from=MaterialExpressionNamedRerouteUsage_107}
   ]
  }
  MaterialExpressionLinearInterpolate_5 {
   class=MaterialExpressionLinearInterpolate
   inputs [
    A {from=MaterialExpressionLinearInterpolate_4}
    B {from=MaterialExpressionNamedRerouteUsage_116}
    Alpha {from=MaterialExpressionNamedRerouteUsage_108}
   ]
  }
  MaterialExpressionLinearInterpolate_6 {
   class=MaterialExpressionLinearInterpolate
   inputs [
    A {from=MaterialExpressionGetMaterialAttributes_7, output=EmissiveColor}
    B {from=MaterialExpressionMaterialFunctionCall_28, output="Blade FX (V3)"}
    Alpha {from=MaterialExpressionMultiply_17}
   ]
  }
  MaterialExpressionMaterialFunctionCall_0 {
   class=MaterialExpressionMaterialFunctionCall
   inputs [
    Material In {from=MaterialExpressionStaticSwitchParameter_14}
    Fresnel {from=MaterialExpressionNamedRerouteUsage_353}
   ]
   properties {materialFunction=&/MF_PixieParts_ThinFilm}
  }
  MaterialExpressionMaterialFunctionCall_10 {
   class=MaterialExpressionMaterialFunctionCall
   inputs [Material In {from=MaterialExpressionNamedRerouteUsage_102}]
   properties {materialFunction=&/MF_PixieParts_DecalMaterials}
  }
  MaterialExpressionMaterialFunctionCall_11 {
   class=MaterialExpressionMaterialFunctionCall
   inputs [
    Base Layer Material In {from=MaterialExpressionNamedRerouteUsage_134}
    Material 1 In {from=MaterialExpressionNamedRerouteUsage_132}
    Mask Material 1 {from=MaterialExpressionNamedRerouteUsage_130}
    Material 2 In {from=MaterialExpressionNamedRerouteUsage_137}
    Mask Material 2 {from=MaterialExpressionNamedRerouteUsage_136}
    Material 3 In {from=MaterialExpressionNamedRerouteUsage_139}
    Mask Material 3 {from=MaterialExpressionNamedRerouteUsage_138}
    Material 4 In {from=MaterialExpressionNamedRerouteUsage_141}
    Mask Material 4 {from=MaterialExpressionNamedRerouteUsage_140}
    Material 5 In {from=MaterialExpressionNamedRerouteUsage_143}
    Mask Material 5 {from=MaterialExpressionNamedRerouteUsage_142}
    Energy Material In {from=MaterialExpressionNamedRerouteUsage_120}
    Energy Emission {from=MaterialExpressionNamedRerouteUsage_145}
    Mask Energy {from=MaterialExpressionNamedRerouteUsage_144}
    Decal Material 1 In {from=MaterialExpressionNamedRerouteUsage_152}
    Mask Decal 1 {from=MaterialExpressionNamedRerouteUsage_151}
    Decal Material 2 In {from=MaterialExpressionNamedRerouteUsage_150}
    Mask Decal 2 {from=MaterialExpressionNamedRerouteUsage_147}
    Decal Material 3 In {from=MaterialExpressionNamedRerouteUsage_154}
    Mask Decal 3 {from=MaterialExpressionNamedRerouteUsage_153}
   ]
   properties {materialFunction=&/MF_PixieParts_BlendMaterials}
  }
  MaterialExpressionMaterialFunctionCall_13 {
   class=MaterialExpressionMaterialFunctionCall
   inputs [
    Base Layer Material In {from=MaterialExpressionNamedRerouteUsage_156}
    Material 1 In {from=MaterialExpressionNamedRerouteUsage_250}
    Mask Material 1 {from=MaterialExpressionNamedRerouteUsage_155}
    Material 2 In {from=MaterialExpressionNamedRerouteUsage_256}
    Mask Material 2 {from=MaterialExpressionNamedRerouteUsage_157}
    Material 3 In {from=MaterialExpressionNamedRerouteUsage_257}
    Mask Material 3 {from=MaterialExpressionNamedRerouteUsage_158}
    Material 4 In {from=MaterialExpressionNamedRerouteUsage_258}
    Mask Material 4 {from=MaterialExpressionNamedRerouteUsage_167}
    Material 5 In {from=MaterialExpressionNamedRerouteUsage_259}
    Mask Material 5 {from=MaterialExpressionNamedRerouteUsage_174}
    Energy Material In {from=MaterialExpressionNamedRerouteUsage_260}
    Energy Emission {from=MaterialExpressionNamedRerouteUsage_176}
    Mask Energy {from=MaterialExpressionNamedRerouteUsage_175}
    Decal Material 1 In {from=MaterialExpressionNamedRerouteUsage_182}
    Mask Decal 1 {from=MaterialExpressionNamedRerouteUsage_181}
    Decal Material 2 In {from=MaterialExpressionNamedRerouteUsage_180}
    Mask Decal 2 {from=MaterialExpressionNamedRerouteUsage_178}
    Decal Material 3 In {from=MaterialExpressionNamedRerouteUsage_249}
    Mask Decal 3 {from=MaterialExpressionNamedRerouteUsage_187}
   ]
   properties {materialFunction=&MF_PixieParts_BlendMaterials}
  }
  MaterialExpressionMaterialFunctionCall_14 {
   class=MaterialExpressionMaterialFunctionCall
   inputs [
    Low _ PC Low and Med {from=MaterialExpressionConstant_2}
    "High _ XB1, PS4" {from=MaterialExpressionSaturate_7}
    "Epic _ PC High, XSX, PS5" {from=MaterialExpressionSaturate_7}
    -------- {from=""}
    "Mobile Low _ PC Perf, Android" {from=MaterialExpressionConstant_2}
    Mobile Med _ Switch {from=MaterialExpressionSaturate_7}
   ]
   properties {materialFunction=/Game/Global/MaterialFunctions/TechArt/MF_Fortnite_QualitySwitch_Scalar}
  }
  MaterialExpressionMaterialFunctionCall_2 {
   class=MaterialExpressionMaterialFunctionCall
   inputs [
    Base Material In {from=MaterialExpressionNamedRerouteUsage_352}
    Armor Material In {from=MaterialExpressionNamedRerouteUsage_39}
    Stone Material In {from=MaterialExpressionNamedRerouteUsage_46}
    Enable Material ID Overrides {from=MaterialExpressionNamedRerouteUsage_323}
    Slot 1 - Material ID {from=MaterialExpressionNamedRerouteUsage_41}
    Slot 2 - Material ID {from=MaterialExpressionNamedRerouteUsage_42}
    Slot 3 - Material ID {from=MaterialExpressionNamedRerouteUsage_43}
    Slot 4 - Material ID {from=MaterialExpressionNamedRerouteUsage_44}
    Slot 5 - Material ID {from=MaterialExpressionNamedRerouteUsage_45}
    AO {from=MaterialExpressionNamedRerouteUsage_269}
    Mask Energy {from=MaterialExpressionNamedRerouteUsage_351}
    Enable Visor {from=MaterialExpressionNamedRerouteUsage_10}
    Visor Override Mask {from=MaterialExpressionNamedRerouteUsage_5}
   ]
   properties {materialFunction=&/MF_PixieParts_MaterialVariants}
  }
  MaterialExpressionMaterialFunctionCall_28 {
   class=MaterialExpressionMaterialFunctionCall
   inputs []
   properties {materialFunction=&/MF_PixieParts_BladeFX}
  }
  MaterialExpressionMaterialFunctionCall_29 {
   class=MaterialExpressionMaterialFunctionCall
   inputs [
    Material In {from=MaterialExpressionNamedRerouteUsage_340}
    Moss Mask In {from=MaterialExpressionNamedRerouteUsage_339}
   ]
   properties {materialFunction=&/MF_PixieParts_MossMaterial}
  }
  MaterialExpressionMaterialFunctionCall_3 {
   class=MaterialExpressionMaterialFunctionCall
   inputs [Material In {from=MaterialExpressionMaterialFunctionCall_0, output="Material Out (MA)"}]
   properties {materialFunction=&/MF_PixieParts_ClearCoat}
  }
  MaterialExpressionMaterialFunctionCall_30 {
   class=MaterialExpressionMaterialFunctionCall
   inputs [
    Base Material In {from=MaterialExpressionNamedRerouteUsage_349}
    Stone Material In {from=MaterialExpressionNamedRerouteUsage_348}
    Decals Mask In {from=MaterialExpressionNamedRerouteUsage_1}
   ]
   properties {materialFunction=&/MF_PixieParts_BaseMaterials}
  }
  MaterialExpressionMaterialFunctionCall_32 {
   class=MaterialExpressionMaterialFunctionCall
   inputs [
    Low _ PC Low and Med {from=MaterialExpressionReroute_9}
    "High _ XB1, PS4" {from=MaterialExpressionMaterialFunctionCall_4, output="Material Out (MA)"}
    "Epic _ PC High, XSX, PS5" {from=MaterialExpressionMaterialFunctionCall_4, output="Material Out (MA)"}
    -------- {from=""}
    "Mobile Low _ PC Perf, Android" {from=MaterialExpressionReroute_9}
    Mobile Med _ Switch {from=MaterialExpressionMaterialFunctionCall_4, output="Material Out (MA)"}
   ]
   properties {materialFunction=/Game/Global/MaterialFunctions/TechArt/MF_Fortnite_QualitySwitch_MaterialAttributes}
  }
  MaterialExpressionMaterialFunctionCall_4 {
   class=MaterialExpressionMaterialFunctionCall
   inputs [Material In {from=MaterialExpressionMaterialFunctionCall_3, output="Material Out (MA)"}]
   properties {materialFunction=&/MF_PixieParts_MetalLUT}
  }
  MaterialExpressionMaterialFunctionCall_5 {
   class=MaterialExpressionMaterialFunctionCall
   inputs [
    Material 1 In {from=MaterialExpressionNamedRerouteUsage_97}
    Material 2 In {from=MaterialExpressionNamedRerouteUsage_98}
    Material 3 In {from=MaterialExpressionNamedRerouteUsage_99}
    Material 4 In {from=MaterialExpressionNamedRerouteUsage_100}
    Material 5 In {from=MaterialExpressionNamedRerouteUsage_342}
    Special Material 1 In {from=MaterialExpressionNamedRerouteUsage_93}
    Special Material 2 In {from=MaterialExpressionNamedRerouteUsage_96}
    Special Material 3 In {from=MaterialExpressionNamedRerouteUsage_94}
    Special Material 4 In {from=MaterialExpressionNamedRerouteUsage_95}
    Material 1 ID {from=MaterialExpressionSubtract_0}
    Material 2 ID {from=MaterialExpressionReroute_2}
    Material 3 ID {from=MaterialExpressionSubtract_4}
    Material 4 ID {from=MaterialExpressionReroute_6}
    Material 5 ID {from=MaterialExpressionSubtract_3}
   ]
   properties {materialFunction=&/MF_PixieParts_SwizzleSpecialMaterials}
  }
  MaterialExpressionMaterialFunctionCall_6 {
   class=MaterialExpressionMaterialFunctionCall
   inputs [
    Material In {from=MaterialExpressionNamedRerouteUsage_82}
    GMap In {from=MaterialExpressionNamedRerouteUsage_103}
    Decals Mask In {from=MaterialExpressionNamedRerouteUsage_105}
   ]
   properties {materialFunction=&/MF_PixieParts_GoldMaterial}
  }
  MaterialExpressionMaterialFunctionCall_7 {
   class=MaterialExpressionMaterialFunctionCall
   inputs [
    Material In {from=MaterialExpressionNamedRerouteUsage_84}
    Fresnel In {from=MaterialExpressionNamedRerouteUsage_119}
   ]
   properties {materialFunction=&/MF_PixieParts_GalaxyMaterial}
  }
  MaterialExpressionMaterialFunctionCall_8 {
   class=MaterialExpressionMaterialFunctionCall
   inputs [
    Material In {from=MaterialExpressionNamedRerouteUsage_121}
    GMap In {from=MaterialExpressionNamedRerouteUsage_6}
   ]
   properties {materialFunction=&/MF_PixieParts_CrystalMaterial}
  }
  MaterialExpressionMaterialFunctionCall_9 {
   class=MaterialExpressionMaterialFunctionCall
   inputs [
    Material In {from=MaterialExpressionNamedRerouteUsage_85}
    GMap In {from=MaterialExpressionNamedRerouteUsage_125}
    Decals Mask In {from=MaterialExpressionNamedRerouteUsage_126}
    Fresnel In {from=MaterialExpressionNamedRerouteUsage_127}
   ]
   properties {materialFunction=&/MF_PixieParts_PearlMaterial}
  }
  MaterialExpressionMax_1 {
   class=MaterialExpressionMax
   inputs [
    A {from=MaterialExpressionConvert_22}
    B {from=MaterialExpressionMax_2}
   ]
  }
  MaterialExpressionMax_2 {
   class=MaterialExpressionMax
   inputs [
    A {from=MaterialExpressionConvert_22}
    B {from=MaterialExpressionConvert_22}
   ]
  }
  MaterialExpressionMultiply_0 {
   class=MaterialExpressionMultiply
   inputs [
    A {from=MaterialExpressionNamedRerouteUsage_8}
    B {from=MaterialExpressionSwitch_0}
   ]
  }
  MaterialExpressionMultiply_1 {
   class=MaterialExpressionMultiply
   inputs [
    A {from=MaterialExpressionChannelMaskParameter_2, output=R}
    B {from=MaterialExpressionNamedRerouteUsage_4}
   ]
  }
  MaterialExpressionMultiply_14 {
   class=MaterialExpressionMultiply
   inputs [
    A {from=MaterialExpressionNamedRerouteUsage_32}
    B {from=MaterialExpressionNamedRerouteUsage_0}
   ]
  }
  MaterialExpressionMultiply_17 {
   class=MaterialExpressionMultiply
   inputs [
    A {from=MaterialExpressionMaterialFunctionCall_28, output="Blade Mask (S)"}
    B {from=MaterialExpressionMultiply_20}
   ]
  }
  MaterialExpressionMultiply_20 {
   class=MaterialExpressionMultiply
   inputs [
    A {from=MaterialExpressionScalarParameter_25}
    B {from=MaterialExpressionNamedRerouteUsage_295}
   ]
  }
  MaterialExpressionMultiply_21 {
   class=MaterialExpressionMultiply
   inputs [
    A {from=MaterialExpressionSaturate_20}
    B {from=MaterialExpressionNamedRerouteUsage_334}
   ]
  }
  MaterialExpressionMultiply_22 {
   class=MaterialExpressionMultiply
   inputs [
    A {from=MaterialExpressionChannelMaskParameter_1, output=R}
    B {from=MaterialExpressionNamedRerouteUsage_3}
   ]
  }
  MaterialExpressionMultiply_23 {
   class=MaterialExpressionMultiply
   inputs [
    A {from=MaterialExpressionNamedRerouteUsage_337}
    B {from=MaterialExpressionNamedRerouteUsage_335}
   ]
  }
  MaterialExpressionMultiply_24 {
   class=MaterialExpressionMultiply
   inputs [
    A {from=MaterialExpressionFunctionInput_5}
    B {from=MaterialExpressionNamedRerouteUsage_2}
   ]
  }
  MaterialExpressionMultiply_25 {
   class=MaterialExpressionMultiply
   inputs [
    A {from=MaterialExpressionMultiply_23}
    B {from=MaterialExpressionOneMinus_8}
   ]
  }
  MaterialExpressionMultiply_27 {
   class=MaterialExpressionMultiply
   inputs [
    A {from=MaterialExpressionLinearInterpolate_6}
    B {from=MaterialExpressionScalarParameter_27}
   ]
  }
  MaterialExpressionNamedRerouteDeclaration_0 {
   class=MaterialExpressionNamedRerouteDeclaration
   inputs [None {from=MaterialExpressionOneMinus_0}]
   properties {name="Armor Materials Mask (S)", nodeColor=<rgba 0.6172 0 1 1>}
  }
  MaterialExpressionNamedRerouteDeclaration_10 {
   class=MaterialExpressionNamedRerouteDeclaration
   inputs [None {from=MaterialExpressionConvert_1}]
   properties {name="Mask Material 4 (S)", nodeColor=<rgba 0 0.4819 1 1>}
  }
  MaterialExpressionNamedRerouteDeclaration_106 {
   class=MaterialExpressionNamedRerouteDeclaration
   inputs [None {from=MaterialExpressionMaterialFunctionCall_14, output=Result}]
   properties {name="Enable Visor (S)", nodeColor=<rgba 1 0 0 1>}
  }
  MaterialExpressionNamedRerouteDeclaration_11 {
   class=MaterialExpressionNamedRerouteDeclaration
   inputs [None {from=MaterialExpressionConvert_1}]
   properties {name="Mask Material 5 (S)", nodeColor=<rgba 0 0.4819 1 1>}
  }
  MaterialExpressionNamedRerouteDeclaration_12 {
   class=MaterialExpressionNamedRerouteDeclaration
   inputs [None {from=MaterialExpressionMaterialFunctionCall_2, output="Mask Energy Out (S)"}]
   properties {name="Mask Energy (S)", nodeColor=<rgba 0 0.4819 1 1>}
  }
  MaterialExpressionNamedRerouteDeclaration_125 {
   class=MaterialExpressionNamedRerouteDeclaration
   inputs [None {from=MaterialExpressionMaterialFunctionCall_30, output="Base Layer Material Out (MA)"}]
   properties {name="Base Layer Material (MA)", nodeColor=<rgba 0.4531 1 0 1>}
  }
  MaterialExpressionNamedRerouteDeclaration_13 {
   class=MaterialExpressionNamedRerouteDeclaration
   inputs [None {from=MaterialExpressionMaterialFunctionCall_2, output="Energy Emission Out (V3)"}]
   properties {name="Energy Emission (V3)", nodeColor=<rgba 1 0 0 1>}
  }
  MaterialExpressionNamedRerouteDeclaration_141 {
   class=MaterialExpressionNamedRerouteDeclaration
   inputs [None {from=MaterialExpressionMaterialFunctionCall_6, output="Material Out (MA)"}]
   properties {name="Gold Material (MA)", nodeColor=<rgba 1 1 1 1>}
  }
  MaterialExpressionNamedRerouteDeclaration_142 {
   class=MaterialExpressionNamedRerouteDeclaration
   inputs [None {from=MaterialExpressionMaterialFunctionCall_2, output="Material 1 Out (MA)"}]
   properties {name="Standard Material 1 (MA)", nodeColor=<rgba 1 1 1 1>}
  }
  MaterialExpressionNamedRerouteDeclaration_151 {
   class=MaterialExpressionNamedRerouteDeclaration
   inputs [None {from=MaterialExpressionMaterialFunctionCall_2, output="Material 2 Out (MA)"}]
   properties {name="Standard Material 2 (MA)", nodeColor=<rgba 1 1 1 1>}
  }
  MaterialExpressionNamedRerouteDeclaration_152 {
   class=MaterialExpressionNamedRerouteDeclaration
   inputs [None {from=MaterialExpressionMaterialFunctionCall_2, output="Material 4 Out (MA)"}]
   properties {name="Standard Material 4 (MA)", nodeColor=<rgba 1 1 1 1>}
  }
  MaterialExpressionNamedRerouteDeclaration_153 {
   class=MaterialExpressionNamedRerouteDeclaration
   inputs [None {from=MaterialExpressionMaterialFunctionCall_2, output="Material 3 Out (MA)"}]
   properties {name="Standard Material 3 (MA)", nodeColor=<rgba 1 1 1 1>}
  }
  MaterialExpressionNamedRerouteDeclaration_167 {
   class=MaterialExpressionNamedRerouteDeclaration
   inputs [None {from=MaterialExpressionMaterialFunctionCall_2, output="Material 5 Out (MA)"}]
   properties {name="Standard Material 5 (MA)", nodeColor=<rgba 1 1 1 1>}
  }
  MaterialExpressionNamedRerouteDeclaration_169 {
   class=MaterialExpressionNamedRerouteDeclaration
   inputs [None {from=MaterialExpressionConvert_23}]
   properties {name="GMap In (S)", nodeColor=<rgba 1 0.4174 0 1>}
  }
  MaterialExpressionNamedRerouteDeclaration_171 {
   class=MaterialExpressionNamedRerouteDeclaration
   inputs [None {from=MaterialExpressionMax_1}]
   properties {name="Decals Coverage (S)", nodeColor=<rgba 0 1 0.1524 1>}
  }
  MaterialExpressionNamedRerouteDeclaration_172 {
   class=MaterialExpressionNamedRerouteDeclaration
   inputs [None {from=MaterialExpressionOneMinus_3}]
   properties {name="Standard Materials Mask (S)", nodeColor=<rgba 0.5 0 1 1>}
  }
  MaterialExpressionNamedRerouteDeclaration_173 {
   class=MaterialExpressionNamedRerouteDeclaration
   inputs [None {from=MaterialExpressionLinearInterpolate_5}]
   properties {name="Materials ID (S)", nodeColor=<rgba 0.9688 0 1 1>}
  }
  MaterialExpressionNamedRerouteDeclaration_174 {
   class=MaterialExpressionNamedRerouteDeclaration
   inputs [None {from=MaterialExpressionFunctionInput_7}]
   properties {name="Fresnel In (S)", nodeColor=<rgba 0 1 0.0156 1>}
  }
  MaterialExpressionNamedRerouteDeclaration_175 {
   class=MaterialExpressionNamedRerouteDeclaration
   inputs [None {from=MaterialExpressionMaterialFunctionCall_2, output="Energy Material Out (V3)"}]
   properties {name="Energy Material (MA)", nodeColor=<rgba 1 1 1 1>}
  }
  MaterialExpressionNamedRerouteDeclaration_177 {
   class=MaterialExpressionNamedRerouteDeclaration
   inputs [None {from=MaterialExpressionMaterialFunctionCall_10, output="Decal Material 2 Out (MA)"}]
   properties {name="Decal Material 2 (MA)", nodeColor=<rgba 0.4531 1 0 1>}
  }
  MaterialExpressionNamedRerouteDeclaration_178 {
   class=MaterialExpressionNamedRerouteDeclaration
   inputs [None {from=MaterialExpressionMaterialFunctionCall_10, output="Decal Material 3 Out (MA)"}]
   properties {name="Decal Material 3 (MA)", nodeColor=<rgba 0.4531 1 0 1>}
  }
  MaterialExpressionNamedRerouteDeclaration_179 {
   class=MaterialExpressionNamedRerouteDeclaration
   inputs [None {from=MaterialExpressionMaterialFunctionCall_13, output="Material Out (MA)"}]
   properties {name="Standard Material Out (MA)", nodeColor=<rgba 1 1 1 1>}
  }
  MaterialExpressionNamedRerouteDeclaration_182 {
   class=MaterialExpressionNamedRerouteDeclaration
   inputs [None {from=MaterialExpressionConvert_23}]
   properties {name="AO In (S)", nodeColor=<rgba 1 0.4174 0 1>}
  }
  MaterialExpressionNamedRerouteDeclaration_183 {
   class=MaterialExpressionNamedRerouteDeclaration
   inputs [None {from=MaterialExpressionMaterialFunctionCall_29, output="Material Out (MA)"}]
   properties {name="Moss Material (MA)", nodeColor=<rgba 0.4531 1 0 1>}
  }
  MaterialExpressionNamedRerouteDeclaration_185 {
   class=MaterialExpressionNamedRerouteDeclaration
   inputs [None {from=MaterialExpressionMultiply_21}]
   properties {name="Stone Materials Mask (S)", nodeColor=<rgba 0.6172 0 1 1>}
  }
  MaterialExpressionNamedRerouteDeclaration_186 {
   class=MaterialExpressionNamedRerouteDeclaration
   inputs [None {from=MaterialExpressionMultiply_22}]
   properties {name="Moss Mask In (S)", nodeColor=<rgba 0 1 0.0391 1>}
  }
  MaterialExpressionNamedRerouteDeclaration_187 {
   class=MaterialExpressionNamedRerouteDeclaration
   inputs [None {from=MaterialExpressionMaterialFunctionCall_29, output="Moss Mask Out (MA)"}]
   properties {name="Moss Mask Out (S)", nodeColor=<rgba 0 1 0.0391 1>}
  }
  MaterialExpressionNamedRerouteDeclaration_188 {
   class=MaterialExpressionNamedRerouteDeclaration
   inputs [None {from=MaterialExpressionMultiply_25}]
   properties {name="Moss Mask (S)", nodeColor=<rgba 0 1 0.0391 1>}
  }
  MaterialExpressionNamedRerouteDeclaration_189 {
   class=MaterialExpressionNamedRerouteDeclaration
   inputs [None {from=MaterialExpressionConvert_1}]
   properties {name="Mask Energy In (S)", nodeColor=<rgba 0 0.4819 1 1>}
  }
  MaterialExpressionNamedRerouteDeclaration_190 {
   class=MaterialExpressionNamedRerouteDeclaration
   inputs [None {from=MaterialExpressionBlendMaterialAttributes_2}]
   properties {name="Material Out (Pre-Animated FX) (MA)", nodeColor=<rgba 1 1 1 1>}
  }
  MaterialExpressionNamedRerouteDeclaration_2 {
   class=MaterialExpressionNamedRerouteDeclaration
   inputs [None {from=MaterialExpressionConvert_6}]
   properties {name="Enable Decals (S)", nodeColor=<rgba 0 0.6719 1 1>}
  }
  MaterialExpressionNamedRerouteDeclaration_23 {
   class=MaterialExpressionNamedRerouteDeclaration
   inputs [None {from=MaterialExpressionMultiply_24}]
   properties {name="Decals Mask In (V3)", nodeColor=<rgba 0 1 0.1524 1>}
  }
  MaterialExpressionNamedRerouteDeclaration_28 {
   class=MaterialExpressionNamedRerouteDeclaration
   inputs [None {from=MaterialExpressionMaterialFunctionCall_30, output="Armor Material Out (MA)"}]
   properties {name="Armor Material (MA)", nodeColor=<rgba 0.4531 1 0 1>}
  }
  MaterialExpressionNamedRerouteDeclaration_29 {
   class=MaterialExpressionNamedRerouteDeclaration
   inputs [None {from=MaterialExpressionFunctionInput_3}]
   properties {name="Materials Mask 2 (V3)", nodeColor=<rgba 0 0.4819 1 1>}
  }
  MaterialExpressionNamedRerouteDeclaration_3 {
   class=MaterialExpressionNamedRerouteDeclaration
   inputs [None {from=MaterialExpressionConvert_6}]
   properties {name="Enable Moss (S)", nodeColor=<rgba 0 0.6719 1 1>}
  }
  MaterialExpressionNamedRerouteDeclaration_30 {
   class=MaterialExpressionNamedRerouteDeclaration
   inputs [None {from=MaterialExpressionFunctionInput_1}]
   properties {name="Materials Mask 1 (V3)", nodeColor=<rgba 0.5703 0 1 1>}
  }
  MaterialExpressionNamedRerouteDeclaration_32 {
   class=MaterialExpressionNamedRerouteDeclaration
   inputs [None {from=MaterialExpressionConvert_17}]
   properties {name="Mask Decal 1 (S)", nodeColor=<rgba 0 1 0.1524 1>}
  }
  MaterialExpressionNamedRerouteDeclaration_33 {
   class=MaterialExpressionNamedRerouteDeclaration
   inputs [None {from=MaterialExpressionConvert_17}]
   properties {name="Mask Decal 2 (S)", nodeColor=<rgba 0 1 0.1524 1>}
  }
  MaterialExpressionNamedRerouteDeclaration_34 {
   class=MaterialExpressionNamedRerouteDeclaration
   inputs [None {from=MaterialExpressionConvert_17}]
   properties {name="Mask Decal 3 (S)", nodeColor=<rgba 0 1 0.1524 1>}
  }
  MaterialExpressionNamedRerouteDeclaration_35 {
   class=MaterialExpressionNamedRerouteDeclaration
   inputs [None {from=MaterialExpressionMaterialFunctionCall_10, output="Decal Material 1 Out (MA)"}]
   properties {name="Decal Material 1 (MA)", nodeColor=<rgba 0.4531 1 0 1>}
  }
  MaterialExpressionNamedRerouteDeclaration_39 {
   class=MaterialExpressionNamedRerouteDeclaration
   inputs [None {from=MaterialExpressionMaterialFunctionCall_5, output="Material 1 Out (MA)"}]
   properties {name="Material 1 Out (MA)", nodeColor=<rgba 1 1 1 1>}
  }
  MaterialExpressionNamedRerouteDeclaration_4 {
   class=MaterialExpressionNamedRerouteDeclaration
   inputs [None {from=MaterialExpressionConvert_6}]
   properties {name="Enable Visor Override (S)", nodeColor=<rgba 0 0.6719 1 1>}
  }
  MaterialExpressionNamedRerouteDeclaration_40 {
   class=MaterialExpressionNamedRerouteDeclaration
   inputs [None {from=MaterialExpressionMaterialFunctionCall_5, output="Material 2 Out (MA)"}]
   properties {name="Material 2 Out (MA)", nodeColor=<rgba 1 1 1 1>}
  }
  MaterialExpressionNamedRerouteDeclaration_41 {
   class=MaterialExpressionNamedRerouteDeclaration
   inputs [None {from=MaterialExpressionMaterialFunctionCall_5, output="Material 4 Out (MA)"}]
   properties {name="Material 4 Out (MA)", nodeColor=<rgba 1 1 1 1>}
  }
  MaterialExpressionNamedRerouteDeclaration_42 {
   class=MaterialExpressionNamedRerouteDeclaration
   inputs [None {from=MaterialExpressionMaterialFunctionCall_5, output="Material 3 Out (MA)"}]
   properties {name="Material 3 Out (MA)", nodeColor=<rgba 1 1 1 1>}
  }
  MaterialExpressionNamedRerouteDeclaration_43 {
   class=MaterialExpressionNamedRerouteDeclaration
   inputs [None {from=MaterialExpressionMaterialFunctionCall_5, output="Material 5 Out (MA)"}]
   properties {name="Material 5 Out (MA)", nodeColor=<rgba 1 1 1 1>}
  }
  MaterialExpressionNamedRerouteDeclaration_46 {
   class=MaterialExpressionNamedRerouteDeclaration
   inputs [None {from=MaterialExpressionMaterialFunctionCall_7, output="Material Out (MA)"}]
   properties {name="Galaxy Material (MA)", nodeColor=<rgba 1 1 1 1>}
  }
  MaterialExpressionNamedRerouteDeclaration_5 {
   class=MaterialExpressionNamedRerouteDeclaration
   inputs [None {from=MaterialExpressionPower_2}]
   properties {name="Visor Override Mask In (S)", nodeColor=<rgba 0 1 0.0391 1>}
  }
  MaterialExpressionNamedRerouteDeclaration_52 {
   class=MaterialExpressionNamedRerouteDeclaration
   inputs [None {from=MaterialExpressionMaterialFunctionCall_9, output="Material Out (MA)"}]
   properties {name="Special Material 4 (MA)", nodeColor=<rgba 1 1 1 1>}
  }
  MaterialExpressionNamedRerouteDeclaration_53 {
   class=MaterialExpressionNamedRerouteDeclaration
   inputs [None {from=MaterialExpressionMaterialFunctionCall_8, output="Material Out (MA)"}]
   properties {name="Special Material 3 (MA)", nodeColor=<rgba 1 1 1 1>}
  }
  MaterialExpressionNamedRerouteDeclaration_54 {
   class=MaterialExpressionNamedRerouteDeclaration
   inputs [None {from=MaterialExpressionScalarParameter_21}]
   properties {name="Body Material ID (S)", nodeColor=<rgba 0 0.4174 1 1>}
  }
  MaterialExpressionNamedRerouteDeclaration_55 {
   class=MaterialExpressionNamedRerouteDeclaration
   inputs [None {from=MaterialExpressionConvert_20}]
   properties {name="FXMask Material ID Overrides (V3)", nodeColor=<rgba 0.0174 1 0 1>}
  }
  MaterialExpressionNamedRerouteDeclaration_56 {
   class=MaterialExpressionNamedRerouteDeclaration
   inputs [None {from=MaterialExpressionConvert_24}]
   properties {name="FX1Mask Material ID Overrides (V3)", nodeColor=<rgba 0.0174 1 0 1>}
  }
  MaterialExpressionNamedRerouteDeclaration_57 {
   class=MaterialExpressionNamedRerouteDeclaration
   inputs [None {from=MaterialExpressionConvert_19}]
   properties {name="Slot 1 - Material ID (S)", nodeColor=<rgba 1 0 0 1>}
  }
  MaterialExpressionNamedRerouteDeclaration_61 {
   class=MaterialExpressionNamedRerouteDeclaration
   inputs [None {from=MaterialExpressionConvert_19}]
   properties {name="Slot 2 - Material ID (S)", nodeColor=<rgba 1 0 0 1>}
  }
  MaterialExpressionNamedRerouteDeclaration_62 {
   class=MaterialExpressionNamedRerouteDeclaration
   inputs [None {from=MaterialExpressionConvert_19}]
   properties {name="Slot 3 - Material ID (S)", nodeColor=<rgba 1 0 0 1>}
  }
  MaterialExpressionNamedRerouteDeclaration_63 {
   class=MaterialExpressionNamedRerouteDeclaration
   inputs [None {from=MaterialExpressionConvert_25}]
   properties {name="Slot 4 - Material ID (S)", nodeColor=<rgba 1 0 0 1>}
  }
  MaterialExpressionNamedRerouteDeclaration_64 {
   class=MaterialExpressionNamedRerouteDeclaration
   inputs [None {from=MaterialExpressionConvert_25}]
   properties {name="Slot 5 - Material ID (S)", nodeColor=<rgba 1 0 0 1>}
  }
  MaterialExpressionNamedRerouteDeclaration_65 {
   class=MaterialExpressionNamedRerouteDeclaration
   inputs [None {from=MaterialExpressionFunctionInput_0}]
   properties {name="Base Material In (MA)", nodeColor=<rgba 0.4531 1 0 1>}
  }
  MaterialExpressionNamedRerouteDeclaration_66 {
   class=MaterialExpressionNamedRerouteDeclaration
   inputs [None {from=MaterialExpressionFunctionInput_8}]
   properties {name="Stone Material In (MA)", nodeColor=<rgba 0.4531 1 0 1>}
  }
  MaterialExpressionNamedRerouteDeclaration_67 {
   class=MaterialExpressionNamedRerouteDeclaration
   inputs [None {from=MaterialExpressionMaterialFunctionCall_30, output="Stone Material Out (MA)"}]
   properties {name="Stone Material (MA)", nodeColor=<rgba 0.4531 1 0 1>}
  }
  MaterialExpressionNamedRerouteDeclaration_69 {
   class=MaterialExpressionNamedRerouteDeclaration
   inputs [None {from=MaterialExpressionBlendMaterialAttributes_0}]
   properties {name="Complex Material Out (MA)", nodeColor=<rgba 1 1 1 1>}
  }
  MaterialExpressionNamedRerouteDeclaration_7 {
   class=MaterialExpressionNamedRerouteDeclaration
   inputs [None {from=MaterialExpressionConvert_0}]
   properties {name="Mask Material 1 (S)", nodeColor=<rgba 0.5703 0 1 1>}
  }
  MaterialExpressionNamedRerouteDeclaration_74 {
   class=MaterialExpressionNamedRerouteDeclaration
   inputs [None {from=MaterialExpressionRound_9}]
   properties {name="Enable Material ID Overrides (S)", nodeColor=<rgba 0 0.4174 1 1>}
  }
  MaterialExpressionNamedRerouteDeclaration_8 {
   class=MaterialExpressionNamedRerouteDeclaration
   inputs [None {from=MaterialExpressionConvert_0}]
   properties {name="Mask Material 2 (S)", nodeColor=<rgba 0.5703 0 1 1>}
  }
  MaterialExpressionNamedRerouteDeclaration_84 {
   class=MaterialExpressionNamedRerouteDeclaration
   inputs [None {from=MaterialExpressionConvert_21}]
   properties {name="FXMask Body Material IDs (V3)", nodeColor=<rgba 0.0174 1 0 1>}
  }
  MaterialExpressionNamedRerouteDeclaration_88 {
   class=MaterialExpressionNamedRerouteDeclaration
   inputs [None {from=MaterialExpressionConvert_21}]
   properties {name="FX1Mask Body Material IDs (V2)", nodeColor=<rgba 0.0174 1 0 1>}
  }
  MaterialExpressionNamedRerouteDeclaration_9 {
   class=MaterialExpressionNamedRerouteDeclaration
   inputs [None {from=MaterialExpressionConvert_0}]
   properties {name="Mask Material 3 (S)", nodeColor=<rgba 0.5703 0 1 1>}
  }
  MaterialExpressionNamedRerouteUsage_0 {class=MaterialExpressionNamedRerouteUsage}
  MaterialExpressionNamedRerouteUsage_1 {class=MaterialExpressionNamedRerouteUsage}
  MaterialExpressionNamedRerouteUsage_10 {class=MaterialExpressionNamedRerouteUsage}
  MaterialExpressionNamedRerouteUsage_100 {class=MaterialExpressionNamedRerouteUsage}
  MaterialExpressionNamedRerouteUsage_102 {class=MaterialExpressionNamedRerouteUsage}
  MaterialExpressionNamedRerouteUsage_103 {class=MaterialExpressionNamedRerouteUsage}
  MaterialExpressionNamedRerouteUsage_104 {class=MaterialExpressionNamedRerouteUsage}
  MaterialExpressionNamedRerouteUsage_105 {class=MaterialExpressionNamedRerouteUsage}
  MaterialExpressionNamedRerouteUsage_106 {class=MaterialExpressionNamedRerouteUsage}
  MaterialExpressionNamedRerouteUsage_107 {class=MaterialExpressionNamedRerouteUsage}
  MaterialExpressionNamedRerouteUsage_108 {class=MaterialExpressionNamedRerouteUsage}
  MaterialExpressionNamedRerouteUsage_109 {class=MaterialExpressionNamedRerouteUsage}
  MaterialExpressionNamedRerouteUsage_11 {class=MaterialExpressionNamedRerouteUsage}
  MaterialExpressionNamedRerouteUsage_110 {class=MaterialExpressionNamedRerouteUsage}
  MaterialExpressionNamedRerouteUsage_112 {class=MaterialExpressionNamedRerouteUsage}
  MaterialExpressionNamedRerouteUsage_113 {class=MaterialExpressionNamedRerouteUsage}
  MaterialExpressionNamedRerouteUsage_114 {class=MaterialExpressionNamedRerouteUsage}
  MaterialExpressionNamedRerouteUsage_115 {class=MaterialExpressionNamedRerouteUsage}
  MaterialExpressionNamedRerouteUsage_116 {class=MaterialExpressionNamedRerouteUsage}
  MaterialExpressionNamedRerouteUsage_119 {class=MaterialExpressionNamedRerouteUsage}
  MaterialExpressionNamedRerouteUsage_120 {class=MaterialExpressionNamedRerouteUsage}
  MaterialExpressionNamedRerouteUsage_121 {class=MaterialExpressionNamedRerouteUsage}
  MaterialExpressionNamedRerouteUsage_125 {class=MaterialExpressionNamedRerouteUsage}
  MaterialExpressionNamedRerouteUsage_126 {class=MaterialExpressionNamedRerouteUsage}
  MaterialExpressionNamedRerouteUsage_127 {class=MaterialExpressionNamedRerouteUsage}
  MaterialExpressionNamedRerouteUsage_130 {class=MaterialExpressionNamedRerouteUsage}
  MaterialExpressionNamedRerouteUsage_132 {class=MaterialExpressionNamedRerouteUsage}
  MaterialExpressionNamedRerouteUsage_134 {class=MaterialExpressionNamedRerouteUsage}
  MaterialExpressionNamedRerouteUsage_136 {class=MaterialExpressionNamedRerouteUsage}
  MaterialExpressionNamedRerouteUsage_137 {class=MaterialExpressionNamedRerouteUsage}
  MaterialExpressionNamedRerouteUsage_138 {class=MaterialExpressionNamedRerouteUsage}
  MaterialExpressionNamedRerouteUsage_139 {class=MaterialExpressionNamedRerouteUsage}
  MaterialExpressionNamedRerouteUsage_140 {class=MaterialExpressionNamedRerouteUsage}
  MaterialExpressionNamedRerouteUsage_141 {class=MaterialExpressionNamedRerouteUsage}
  MaterialExpressionNamedRerouteUsage_142 {class=MaterialExpressionNamedRerouteUsage}
  MaterialExpressionNamedRerouteUsage_143 {class=MaterialExpressionNamedRerouteUsage}
  MaterialExpressionNamedRerouteUsage_144 {class=MaterialExpressionNamedRerouteUsage}
  MaterialExpressionNamedRerouteUsage_145 {class=MaterialExpressionNamedRerouteUsage}
  MaterialExpressionNamedRerouteUsage_147 {class=MaterialExpressionNamedRerouteUsage}
  MaterialExpressionNamedRerouteUsage_150 {class=MaterialExpressionNamedRerouteUsage}
  MaterialExpressionNamedRerouteUsage_151 {class=MaterialExpressionNamedRerouteUsage}
  MaterialExpressionNamedRerouteUsage_152 {class=MaterialExpressionNamedRerouteUsage}
  MaterialExpressionNamedRerouteUsage_153 {class=MaterialExpressionNamedRerouteUsage}
  MaterialExpressionNamedRerouteUsage_154 {class=MaterialExpressionNamedRerouteUsage}
  MaterialExpressionNamedRerouteUsage_155 {class=MaterialExpressionNamedRerouteUsage}
  MaterialExpressionNamedRerouteUsage_156 {class=MaterialExpressionNamedRerouteUsage}
  MaterialExpressionNamedRerouteUsage_157 {class=MaterialExpressionNamedRerouteUsage}
  MaterialExpressionNamedRerouteUsage_158 {class=MaterialExpressionNamedRerouteUsage}
  MaterialExpressionNamedRerouteUsage_167 {class=MaterialExpressionNamedRerouteUsage}
  MaterialExpressionNamedRerouteUsage_174 {class=MaterialExpressionNamedRerouteUsage}
  MaterialExpressionNamedRerouteUsage_175 {class=MaterialExpressionNamedRerouteUsage}
  MaterialExpressionNamedRerouteUsage_176 {class=MaterialExpressionNamedRerouteUsage}
  MaterialExpressionNamedRerouteUsage_178 {class=MaterialExpressionNamedRerouteUsage}
  MaterialExpressionNamedRerouteUsage_180 {class=MaterialExpressionNamedRerouteUsage}
  MaterialExpressionNamedRerouteUsage_181 {class=MaterialExpressionNamedRerouteUsage}
  MaterialExpressionNamedRerouteUsage_182 {class=MaterialExpressionNamedRerouteUsage}
  MaterialExpressionNamedRerouteUsage_187 {class=MaterialExpressionNamedRerouteUsage}
  MaterialExpressionNamedRerouteUsage_2 {class=MaterialExpressionNamedRerouteUsage}
  MaterialExpressionNamedRerouteUsage_249 {class=MaterialExpressionNamedRerouteUsage}
  MaterialExpressionNamedRerouteUsage_250 {class=MaterialExpressionNamedRerouteUsage}
  MaterialExpressionNamedRerouteUsage_256 {class=MaterialExpressionNamedRerouteUsage}
  MaterialExpressionNamedRerouteUsage_257 {class=MaterialExpressionNamedRerouteUsage}
  MaterialExpressionNamedRerouteUsage_258 {class=MaterialExpressionNamedRerouteUsage}
  MaterialExpressionNamedRerouteUsage_259 {class=MaterialExpressionNamedRerouteUsage}
  MaterialExpressionNamedRerouteUsage_260 {class=MaterialExpressionNamedRerouteUsage}
  MaterialExpressionNamedRerouteUsage_269 {class=MaterialExpressionNamedRerouteUsage}
  MaterialExpressionNamedRerouteUsage_273 {class=MaterialExpressionNamedRerouteUsage}
  MaterialExpressionNamedRerouteUsage_283 {class=MaterialExpressionNamedRerouteUsage}
  MaterialExpressionNamedRerouteUsage_295 {class=MaterialExpressionNamedRerouteUsage}
  MaterialExpressionNamedRerouteUsage_3 {class=MaterialExpressionNamedRerouteUsage}
  MaterialExpressionNamedRerouteUsage_32 {class=MaterialExpressionNamedRerouteUsage}
  MaterialExpressionNamedRerouteUsage_323 {class=MaterialExpressionNamedRerouteUsage}
  MaterialExpressionNamedRerouteUsage_325 {class=MaterialExpressionNamedRerouteUsage}
  MaterialExpressionNamedRerouteUsage_33 {class=MaterialExpressionNamedRerouteUsage}
  MaterialExpressionNamedRerouteUsage_334 {class=MaterialExpressionNamedRerouteUsage}
  MaterialExpressionNamedRerouteUsage_335 {class=MaterialExpressionNamedRerouteUsage}
  MaterialExpressionNamedRerouteUsage_337 {class=MaterialExpressionNamedRerouteUsage}
  MaterialExpressionNamedRerouteUsage_339 {class=MaterialExpressionNamedRerouteUsage}
  MaterialExpressionNamedRerouteUsage_34 {class=MaterialExpressionNamedRerouteUsage}
  MaterialExpressionNamedRerouteUsage_340 {class=MaterialExpressionNamedRerouteUsage}
  MaterialExpressionNamedRerouteUsage_342 {class=MaterialExpressionNamedRerouteUsage}
  MaterialExpressionNamedRerouteUsage_344 {class=MaterialExpressionNamedRerouteUsage}
  MaterialExpressionNamedRerouteUsage_346 {class=MaterialExpressionNamedRerouteUsage}
  MaterialExpressionNamedRerouteUsage_347 {class=MaterialExpressionNamedRerouteUsage}
  MaterialExpressionNamedRerouteUsage_348 {class=MaterialExpressionNamedRerouteUsage}
  MaterialExpressionNamedRerouteUsage_349 {class=MaterialExpressionNamedRerouteUsage}
  MaterialExpressionNamedRerouteUsage_351 {class=MaterialExpressionNamedRerouteUsage}
  MaterialExpressionNamedRerouteUsage_352 {class=MaterialExpressionNamedRerouteUsage}
  MaterialExpressionNamedRerouteUsage_353 {class=MaterialExpressionNamedRerouteUsage}
  MaterialExpressionNamedRerouteUsage_354 {class=MaterialExpressionNamedRerouteUsage}
  MaterialExpressionNamedRerouteUsage_39 {class=MaterialExpressionNamedRerouteUsage}
  MaterialExpressionNamedRerouteUsage_4 {class=MaterialExpressionNamedRerouteUsage}
  MaterialExpressionNamedRerouteUsage_41 {class=MaterialExpressionNamedRerouteUsage}
  MaterialExpressionNamedRerouteUsage_42 {class=MaterialExpressionNamedRerouteUsage}
  MaterialExpressionNamedRerouteUsage_43 {class=MaterialExpressionNamedRerouteUsage}
  MaterialExpressionNamedRerouteUsage_44 {class=MaterialExpressionNamedRerouteUsage}
  MaterialExpressionNamedRerouteUsage_45 {class=MaterialExpressionNamedRerouteUsage}
  MaterialExpressionNamedRerouteUsage_46 {class=MaterialExpressionNamedRerouteUsage}
  MaterialExpressionNamedRerouteUsage_5 {class=MaterialExpressionNamedRerouteUsage}
  MaterialExpressionNamedRerouteUsage_56 {class=MaterialExpressionNamedRerouteUsage}
  MaterialExpressionNamedRerouteUsage_57 {class=MaterialExpressionNamedRerouteUsage}
  MaterialExpressionNamedRerouteUsage_58 {class=MaterialExpressionNamedRerouteUsage}
  MaterialExpressionNamedRerouteUsage_59 {class=MaterialExpressionNamedRerouteUsage}
  MaterialExpressionNamedRerouteUsage_6 {class=MaterialExpressionNamedRerouteUsage}
  MaterialExpressionNamedRerouteUsage_60 {class=MaterialExpressionNamedRerouteUsage}
  MaterialExpressionNamedRerouteUsage_7 {class=MaterialExpressionNamedRerouteUsage}
  MaterialExpressionNamedRerouteUsage_8 {class=MaterialExpressionNamedRerouteUsage}
  MaterialExpressionNamedRerouteUsage_82 {class=MaterialExpressionNamedRerouteUsage}
  MaterialExpressionNamedRerouteUsage_84 {class=MaterialExpressionNamedRerouteUsage}
  MaterialExpressionNamedRerouteUsage_85 {class=MaterialExpressionNamedRerouteUsage}
  MaterialExpressionNamedRerouteUsage_88 {class=MaterialExpressionNamedRerouteUsage}
  MaterialExpressionNamedRerouteUsage_89 {class=MaterialExpressionNamedRerouteUsage}
  MaterialExpressionNamedRerouteUsage_9 {class=MaterialExpressionNamedRerouteUsage}
  MaterialExpressionNamedRerouteUsage_90 {class=MaterialExpressionNamedRerouteUsage}
  MaterialExpressionNamedRerouteUsage_91 {class=MaterialExpressionNamedRerouteUsage}
  MaterialExpressionNamedRerouteUsage_92 {class=MaterialExpressionNamedRerouteUsage}
  MaterialExpressionNamedRerouteUsage_93 {class=MaterialExpressionNamedRerouteUsage}
  MaterialExpressionNamedRerouteUsage_94 {class=MaterialExpressionNamedRerouteUsage}
  MaterialExpressionNamedRerouteUsage_95 {class=MaterialExpressionNamedRerouteUsage}
  MaterialExpressionNamedRerouteUsage_96 {class=MaterialExpressionNamedRerouteUsage}
  MaterialExpressionNamedRerouteUsage_97 {class=MaterialExpressionNamedRerouteUsage}
  MaterialExpressionNamedRerouteUsage_98 {class=MaterialExpressionNamedRerouteUsage}
  MaterialExpressionNamedRerouteUsage_99 {class=MaterialExpressionNamedRerouteUsage}
  MaterialExpressionOneMinus_0 {class=MaterialExpressionOneMinus, inputs [None {from=MaterialExpressionSaturate_20}]}
  MaterialExpressionOneMinus_1 {class=MaterialExpressionOneMinus, inputs [None {from=MaterialExpressionNamedRerouteUsage_7}]}
  MaterialExpressionOneMinus_3 {class=MaterialExpressionOneMinus, inputs [None {from=MaterialExpressionSaturate_15}]}
  MaterialExpressionOneMinus_8 {class=MaterialExpressionOneMinus, inputs [None {from=MaterialExpressionNamedRerouteUsage_344}]}
  MaterialExpressionPower_2 {
   class=MaterialExpressionPower
   inputs [
    Base {from=MaterialExpressionMultiply_1}
    Exp {from=""}
   ]
   properties {constExponent=0.2}
  }
  MaterialExpressionReroute_0 {class=MaterialExpressionReroute, inputs [None {from=MaterialExpressionSetMaterialAttributes_2}]}
  MaterialExpressionReroute_2 {class=MaterialExpressionReroute, inputs [None {from=MaterialExpressionSubtract_1}]}
  MaterialExpressionReroute_6 {class=MaterialExpressionReroute, inputs [None {from=MaterialExpressionSubtract_2}]}
  MaterialExpressionReroute_9 {class=MaterialExpressionReroute, inputs [None {from=MaterialExpressionStaticSwitchParameter_14}]}
  MaterialExpressionRound_7 {class=MaterialExpressionRound, inputs [None {from=MaterialExpressionSwitch_9, output=""}], properties {}}
  MaterialExpressionRound_8 {class=MaterialExpressionRound, inputs [None {from=MaterialExpressionSwitch_10, output=""}], properties {}}
  MaterialExpressionRound_9 {class=MaterialExpressionRound, inputs [None {from=MaterialExpressionSaturate_4, output=""}], properties {}}
  MaterialExpressionSaturate_15 {class=MaterialExpressionSaturate, inputs [None {from=MaterialExpressionSubtract_5}]}
  MaterialExpressionSaturate_20 {class=MaterialExpressionSaturate, inputs [None {from=MaterialExpressionNamedRerouteUsage_325}]}
  MaterialExpressionSaturate_4 {class=MaterialExpressionSaturate, inputs [None {from=MaterialExpressionScalarParameter_23}]}
  MaterialExpressionSaturate_7 {class=MaterialExpressionSaturate, inputs [None {from=MaterialExpressionScalarParameter_0}]}
  MaterialExpressionScalarParameter_0 {
   class=MaterialExpressionScalarParameter
   properties {group=PixieParts | Materials | Visor, parameterName=Enable Visor, sliderMax=1, sortPriority=10}
  }
  MaterialExpressionScalarParameter_17 {
   class=MaterialExpressionScalarParameter
   properties {
    desc="\r\n0 : Lava\r\n1 : Bolt\r\n2 : Drill\r\n3 : Nail\r\n4 : Screw\r\n5 : Clay\r\n"
    group=PixieParts | Globals
    parameterName=Material 3 ID Override
    sliderMax=6
    sortPriority=140
   }
  }
  MaterialExpressionScalarParameter_18 {
   class=MaterialExpressionScalarParameter
   properties {
    desc="\r\n0 : Lava\r\n1 : Bolt\r\n2 : Drill\r\n3 : Nail\r\n4 : Screw\r\n5 : Clay\r\n"
    group=PixieParts | Globals
    parameterName=Material 4 ID Override
    sliderMax=6
    sortPriority=150
   }
  }
  MaterialExpressionScalarParameter_20 {
   class=MaterialExpressionScalarParameter
   properties {
    desc="\r\n0 : Lava\r\n1 : Bolt\r\n2 : Drill\r\n3 : Nail\r\n4 : Screw\r\n5 : Clay\r\n"
    group=PixieParts | Globals
    parameterName=Material 5 ID Override
    sliderMax=6
    sortPriority=160
   }
  }
  MaterialExpressionScalarParameter_21 {
   class=MaterialExpressionScalarParameter
   properties {
    desc="\r\n0 : Lava\r\n1 : Bolt\r\n2 : Drill\r\n3 : Nail\r\n4 : Screw\r\n5 : Clay\r\n"
    group=PixieParts | Globals
    parameterName=Body Material ID
    sliderMax=5
    sortPriority=110
   }
  }
  MaterialExpressionScalarParameter_23 {
   class=MaterialExpressionScalarParameter
   properties {group=PixieParts | Globals, parameterName=Enable Variant ID Override, sliderMax=1, sortPriority=100}
  }
  MaterialExpressionScalarParameter_25 {class=MaterialExpressionScalarParameter, properties {group=PixieParts | DEV, parameterName=Pickaxe FX Mix, sliderMax=1, sortPriority=200}}
  MaterialExpressionScalarParameter_26 {
   class=MaterialExpressionScalarParameter
   properties {
    desc="\r\n0 : Decals\r\n1 : Moss\r\n2: Visor Override\r\n"
    group=PixieParts | Textures
    parameterName=FX2Mask Mode
    sliderMax=1
    sortPriority=40
   }
  }
  MaterialExpressionScalarParameter_27 {
   class=MaterialExpressionScalarParameter
   properties {defaultValue=1, group=PixieParts | Materials | Energy, parameterName="Emission Gain [MOBILE]", sortPriority=100}
  }
  MaterialExpressionScalarParameter_4 {
   class=MaterialExpressionScalarParameter
   properties {
    desc="\r\n0 : Lava\r\n1 : Bolt\r\n2 : Drill\r\n3 : Nail\r\n4 : Screw\r\n5 : Clay\r\n"
    group=PixieParts | Globals
    parameterName=Material 1 ID Override
    sliderMax=6
    sortPriority=120
   }
  }
  MaterialExpressionScalarParameter_7 {
   class=MaterialExpressionScalarParameter
   properties {
    desc="\r\n0 : Lava\r\n1 : Bolt\r\n2 : Drill\r\n3 : Nail\r\n4 : Screw\r\n5 : Clay\r\n"
    group=PixieParts | Globals
    parameterName=Material 2 ID Override
    sliderMax=6
    sortPriority=130
   }
  }
  MaterialExpressionSetMaterialAttributes_2 {
   class=MaterialExpressionSetMaterialAttributes
   inputs [
    MaterialAttributes {from=MaterialExpressionGetMaterialAttributes_7, output=MaterialAttributes}
    EmissiveColor {from=MaterialExpressionShadingPathSwitch_1, output=""}
   ]
   properties {attributeSetTypes [#1]}
  }
  MaterialExpressionSetMaterialAttributes_4 {
   class=MaterialExpressionSetMaterialAttributes
   inputs [
    MaterialAttributes {from=MaterialExpressionReroute_0, output=""}
    Displacement {from=MaterialExpressionConstant_3, output=""}
    SubsurfaceColor {from=MaterialExpressionConstant_4, output=""}
    Refraction {from=MaterialExpressionConstant_4, output=""}
    TransmittanceColor {from=MaterialExpressionConstant_5, output=""}
    Tangent {from=MaterialExpressionConvert_4, output=""}
    BentNormal {from=MaterialExpressionConvert_7, output=""}
   ]
   properties {
    desc="revert abused channels\r\nto their default values"
    attributeSetTypes [
     #2
     #3
     #4
     #5
     #6
     #7
    ]
   }
  }
  MaterialExpressionShadingPathSwitch_1 {
   class=MaterialExpressionShadingPathSwitch
   inputs [
    Default {from=MaterialExpressionLinearInterpolate_6, output=""}
    Deferred {from="", output=""}
    Forward {from="", output=""}
    Mobile {from=MaterialExpressionMultiply_27, output=""}
   ]
   properties {}
  }
  MaterialExpressionStaticSwitchParameter_14 {
   class=MaterialExpressionStaticSwitchParameter
   inputs [
    True {from=MaterialExpressionNamedRerouteUsage_283, output=""}
    False {from=MaterialExpressionNamedRerouteUsage_273, output=""}
   ]
   properties {group=PixieParts | Globals, parameterName=Enable Special Materials, sortPriority=10}
  }
  MaterialExpressionSubtract_0 {
   class=MaterialExpressionSubtract
   inputs [
    A {from=MaterialExpressionNamedRerouteUsage_88}
    B {from=""}
   ]
  }
  MaterialExpressionSubtract_1 {
   class=MaterialExpressionSubtract
   inputs [
    A {from=MaterialExpressionNamedRerouteUsage_89}
    B {from=""}
   ]
  }
  MaterialExpressionSubtract_2 {
   class=MaterialExpressionSubtract
   inputs [
    A {from=MaterialExpressionNamedRerouteUsage_91}
    B {from=""}
   ]
  }
  MaterialExpressionSubtract_3 {
   class=MaterialExpressionSubtract
   inputs [
    A {from=MaterialExpressionNamedRerouteUsage_92}
    B {from=""}
   ]
  }
  MaterialExpressionSubtract_4 {
   class=MaterialExpressionSubtract
   inputs [
    A {from=MaterialExpressionNamedRerouteUsage_90}
    B {from=""}
   ]
  }
  MaterialExpressionSubtract_5 {
   class=MaterialExpressionSubtract
   inputs [
    A {from=MaterialExpressionNamedRerouteUsage_325}
    B {from=""}
   ]
  }
  MaterialExpressionSwitch_0 {
   class=MaterialExpressionSwitch
   inputs [
    SwitchValue {from=MaterialExpressionNamedRerouteUsage_11, output=""}
    Default {from="", output=""}
    None {from=MaterialExpressionOneMinus_1, output=""}
    None {from=MaterialExpressionConstant_1, output=""}
   ]
   properties {inputs [
     {
      inputName=None
      input {
       expression=&self:MaterialExpressionOneMinus_1
       inputName=None
       mask=0
       maskA=0
       maskB=0
       maskG=0
       maskR=0
       outputIndex=0
      }
     }
     {
      inputName=None
      input {
       expression=&self:MaterialExpressionConstant_1
       inputName=None
       mask=0
       maskA=0
       maskB=0
       maskG=0
       maskR=0
       outputIndex=0
      }
     }
    ]}
  }
  MaterialExpressionSwitch_10 {
   class=MaterialExpressionSwitch
   inputs [
    SwitchValue {from=MaterialExpressionNamedRerouteUsage_60, output=""}
    Default {from="", output=""}
    OveriideOff {from=MaterialExpressionNamedRerouteUsage_57, output=""}
    OverrideOn {from=MaterialExpressionNamedRerouteUsage_58, output=""}
   ]
   properties {inputs [
     {
      inputName=OveriideOff
      input {
       expression=&self:MaterialExpressionNamedRerouteUsage_57
       inputName=None
       mask=0
       maskA=0
       maskB=0
       maskG=0
       maskR=0
       outputIndex=0
      }
     }
     {
      inputName=OverrideOn
      input {
       expression=&self:MaterialExpressionNamedRerouteUsage_58
       inputName=None
       mask=0
       maskA=0
       maskB=0
       maskG=0
       maskR=0
       outputIndex=0
      }
     }
    ]}
  }
  MaterialExpressionSwitch_2 {
   class=MaterialExpressionSwitch
   inputs [
    SwitchValue {from=MaterialExpressionClamp_0, output=""}
    Default {from="", output=""}
    Decals {from=MaterialExpressionConvert_2, output=""}
    Moss {from=MaterialExpressionConvert_3, output=""}
    Visor {from=MaterialExpressionConvert_5, output=""}
   ]
   properties {inputs [
     {
      inputName=Decals
      input {
       expression=&self:MaterialExpressionConvert_2
       inputName=None
       mask=0
       maskA=0
       maskB=0
       maskG=0
       maskR=0
       outputIndex=0
      }
     }
     {
      inputName=Moss
      input {
       expression=&self:MaterialExpressionConvert_3
       inputName=None
       mask=0
       maskA=0
       maskB=0
       maskG=0
       maskR=0
       outputIndex=0
      }
     }
     {
      inputName=Visor
      input {
       expression=&self:MaterialExpressionConvert_5
       inputName=None
       mask=0
       maskA=0
       maskB=0
       maskG=0
       maskR=0
       outputIndex=0
      }
     }
    ]}
  }
  MaterialExpressionSwitch_9 {
   class=MaterialExpressionSwitch
   inputs [
    SwitchValue {from=MaterialExpressionNamedRerouteUsage_60, output=""}
    Default {from="", output=""}
    OveriideOff {from=MaterialExpressionNamedRerouteUsage_56, output=""}
    OverrideOn {from=MaterialExpressionNamedRerouteUsage_59, output=""}
   ]
   properties {inputs [
     {
      inputName=OveriideOff
      input {
       expression=&self:MaterialExpressionNamedRerouteUsage_56
       inputName=None
       mask=0
       maskA=0
       maskB=0
       maskG=0
       maskR=0
       outputIndex=0
      }
     }
     {
      inputName=OverrideOn
      input {
       expression=&self:MaterialExpressionNamedRerouteUsage_59
       inputName=None
       mask=0
       maskA=0
       maskB=0
       maskG=0
       maskR=0
       outputIndex=0
      }
     }
    ]}
  }
 }}
