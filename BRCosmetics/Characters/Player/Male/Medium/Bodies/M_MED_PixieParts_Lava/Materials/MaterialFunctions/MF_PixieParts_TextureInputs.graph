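# Material function that supplies texture inputs for the PixieParts Lava character body material, outputting base and stone material attributes, masks, pre-skinned position, distance blend and Fresnel.
# Texture parameters feed attribute sets through named reroutes, with a Use Packed Maps static switch selecting channel unpacking.
# CharacterBrightnessandScale_Balance adjusts the base material and supplies distance blend.
# Opacity, CustomData0 and Refraction are zeroed for downstream repurposing; FX2 mask can sample UV1.
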MaterialFunction /BRCosmetics/Characters/Player/Male/Medium/Bodies/M_MED_PixieParts_Lava/Materials/MaterialFunctions/MF_PixieParts_TextureInputs
graph {nodes [
  FunctionOutput(None=NamedRerouteUsage()) {outputName="FX Mask (V3)", sortPriority=20}
  FunctionOutput(None=NamedRerouteUsage()) {outputName="FX1 Mask (V3)", sortPriority=30}
  FunctionOutput(None=NamedRerouteUsage()) {outputName="FX2 Mask (V3)", sortPriority=40}
  n1 = PreSkinnedPosition()
  FunctionOutput(None=ShaderStageSwitch(PixelShader=VertexInterpolator(VS=@n1).PS, VertexShader=@n1)) {outputName="Pre-Skinned Local Position (V3)", sortPriority=60}
  FunctionOutput(None=NamedRerouteUsage()) {outputName="Distance Blend (S)", sortPriority=70}
  FunctionOutput(None=Fresnel(ExponentIn=Constant() {r=1}) {baseReflectFraction=0}) {outputName="Fresnel (S)", sortPriority=80}
  FunctionOutput(None=NamedRerouteUsage()) {outputName="Base Material (MA)"}
  FunctionOutput(None=NamedRerouteUsage()) {outputName="Stone Material (MA)", sortPriority=10}
  MaterialExpressionNamedRerouteUsage_25 = NamedRerouteUsage()
  n2 = Convert(None=@MaterialExpressionNamedRerouteUsage_25) {
   convertInputs [{
     type=Vector3
     expressionInput {
      expression=&self:MaterialExpressionNamedRerouteUsage_25
      inputName=None
      mask=0
      maskA=0
      maskB=0
      maskG=0
      maskR=0
      outputIndex=0
     }
    }]
   convertOutputs [{type=Vector3}]
  }
  FunctionOutput(None=@n2) {outputName="Mask (V3)", sortPriority=50}
  MaterialExpressionTextureCoordinate_0 = TextureCoordinate()
  MaterialExpressionTextureCoordinate_1 = TextureCoordinate() {coordinateIndex=1}
  n3 = Switch!(
   SwitchValue=Saturate(None=ScalarParameter() {group=PixieParts | Textures, parameterName=FX2Mask Use UV1, sliderMax=1, sortPriority=30})
   Default
   UV0=@MaterialExpressionTextureCoordinate_0
   UV1=@MaterialExpressionTextureCoordinate_1
  ) {inputs [
    {
     inputName=UV0
     input {
      expression=&self:MaterialExpressionTextureCoordinate_0
      inputName=None
      mask=0
      maskA=0
      maskB=0
      maskG=0
      maskR=0
      outputIndex=0
     }
    }
    {
     inputName=UV1
     input {
      expression=&self:MaterialExpressionTextureCoordinate_1
      inputName=None
      mask=0
      maskA=0
      maskB=0
      maskG=0
      maskR=0
      outputIndex=0
     }
    }
   ]}
  NamedRerouteDeclaration(None=@n3) {name="FX2 UV (V2)", nodeColor=<rgba 0 1 0.1094 1>}
  n4 = Constant()
  n5 = SetMaterialAttributes!(MaterialAttributes, ShadingModel=ShadingModel(), Normal=NamedRerouteUsage(), Opacity=@n4, CustomData0=@n4, Refraction=@n4) {
   desc="Custom Data 0 and Refraction \r\nMUST be set to 0 here as these \r\nchannels will be abused down \r\nthe line for customization purpose"
   attributeSetTypes [
    #1
    #2
    #3
    #4
    #5
   ]
  }
  NamedRerouteDeclaration(None=@n5) {name="Src Material (MA)", nodeColor=<rgba 0.6172 0 1 1>}
  n6 = TextureSampleParameter2D() {
   group=00_Texture Inputs
   parameterName=Diffuse
   samplerSource=SSM_Wrap_WorldGroupSettings
   sortPriority=0
   texture=/Game/Global/Textures/Default/Blanks/T_MidgreysRGB
  }
  NamedRerouteDeclaration(None=@n6.RGB) {name="Base Diffuse Map (V3)", nodeColor=<rgba 0 1 0.2969 1>}
  n7 = TextureSampleParameter2D() {
   group=PixieParts | Textures
   parameterName=Stone - Diffuse
   samplerSource=SSM_Wrap_WorldGroupSettings
   sortPriority=30
   texture=&T_MidgreysRGB
  }
  NamedRerouteDeclaration(None=@n7.RGB) {name="Stone Diffuse Map (V3)", nodeColor=<rgba 0 1 0.2969 1>}
  n8 = TextureSampleParameter2D() {
   desc="R = Spec , G = Roughness, B= Metal"
   group=00_Texture Inputs
   parameterName=SRM
   samplerSource=SSM_Wrap_WorldGroupSettings
   samplerType=SAMPLERTYPE_Masks
   sortPriority=1
   texture=&T_Fortnite_Default_S
  }
  NamedRerouteDeclaration(None=@n8.RGB) {name="Base SRM Map (V3)", nodeColor=<rgba 0 0.3438 1 1>}
  n9 = TextureSampleParameter2D() {
   desc="R = Spec , G = Roughness, B= Metal"
   group=PixieParts | Textures
   parameterName=Stone - SRM
   samplerSource=SSM_Wrap_WorldGroupSettings
   samplerType=SAMPLERTYPE_Masks
   sortPriority=40
   texture=/Game/Global/Textures/Default/Blanks/T_MidgreyMask
  }
  NamedRerouteDeclaration(None=@n9.RGB) {name="Stone SRM Map (V3)", nodeColor=<rgba 0 0.3438 1 1>}
  n10 = TextureSampleParameter2D() {
   group=00_Texture Inputs
   parameterName=Mask
   samplerSource=SSM_Wrap_WorldGroupSettings
   samplerType=SAMPLERTYPE_Masks
   sortPriority=3
   texture=&T_BlackMask
  }
  NamedRerouteDeclaration(None=@n10.RGB) {name="M Map (V3)", nodeColor=<rgba 1 0 0.9844 1>}
  MaterialExpressionNamedRerouteUsage_19 = NamedRerouteUsage()
  n11 = Convert(None=@MaterialExpressionNamedRerouteUsage_19) {
   convertInputs [{
     type=Vector3
     expressionInput {
      expression=&self:MaterialExpressionNamedRerouteUsage_19
      inputName=None
      mask=0
      maskA=0
      maskB=0
      maskG=0
      maskR=0
      outputIndex=0
     }
    }]
   convertOutputs [{type=Scalar}, {type=Scalar}]
  }
  MaterialExpressionNamedRerouteUsage_21 = NamedRerouteUsage()
  n12 = Convert(None=@MaterialExpressionNamedRerouteUsage_21) {
   convertInputs [{
     type=Vector3
     expressionInput {
      expression=&self:MaterialExpressionNamedRerouteUsage_21
      inputName=None
      mask=0
      maskA=0
      maskB=0
      maskG=0
      maskR=0
      outputIndex=0
     }
    }]
   convertOutputs [{type=Scalar}, {type=Scalar}, {type=Scalar}]
  }
  n13 = SetMaterialAttributes!(MaterialAttributes=NamedRerouteUsage(), BaseColor=@n11, EmissiveColor=@n11, Specular=@n12, Roughness=@n12, Metallic=@n12) {attributeSetTypes [
    #6
    #7
    #8
    #9
    #10
   ]}
  MaterialExpressionNamedRerouteUsage_13 = NamedRerouteUsage()
  n14 = Convert(None=@MaterialExpressionNamedRerouteUsage_13) {
   nodeName=Break Float3
   convertInputs [{
     type=Vector3
     expressionInput {
      expression=&self:MaterialExpressionNamedRerouteUsage_13
      inputName=None
      mask=0
      maskA=0
      maskB=0
      maskG=0
      maskR=0
      outputIndex=0
     }
    }]
   convertOutputs [{type=Scalar}, {type=Scalar}, {type=Scalar}]
  }
  n15 = SetMaterialAttributes(
   MaterialAttributes=NamedRerouteUsage()
   BaseColor=NamedRerouteUsage()
   EmissiveColor=NamedRerouteUsage()
   Specular=@n14
   Roughness=@n14
   Metallic=@n14
  ) {attributeSetTypes [
    #6
    #7
    #8
    #9
    #10
   ]}
  n16 = MaterialFunctionCall!(Material=StaticSwitchParameter(True=@n13, False=@n15) {group=PixieParts | Textures, parameterName=Use Packed Maps, sortPriority=100}) {materialFunction=/Game/Packages/Fortress_SharedMaterials/Base_Material_Functions/CharacterBrightnessandScale_Balance}
  NamedRerouteDeclaration(None=@n16.Result) {name="Base Material (MA)", nodeColor=<rgba 0.4531 1 0 1>}
  n17 = NamedRerouteUsage()
  n18 = ComponentMask(None=NamedRerouteUsage()) {g=true}
  n19 = ComponentMask(None=NamedRerouteUsage()) {b=true}
  n20 = ComponentMask(None=NamedRerouteUsage()) {g=true}
  MaterialExpressionNamedRerouteUsage_14 = NamedRerouteUsage()
  n21 = Convert(None=@MaterialExpressionNamedRerouteUsage_14) {
   nodeName=Break Float3
   convertInputs [{
     type=Vector3
     expressionInput {
      expression=&self:MaterialExpressionNamedRerouteUsage_14
      inputName=None
      mask=0
      maskA=0
      maskB=0
      maskG=0
      maskR=0
      outputIndex=0
     }
    }]
   convertOutputs [{type=Scalar}, {type=Scalar}, {type=Scalar}]
  }
  n22 = SetMaterialAttributes!(MaterialAttributes=NamedRerouteUsage(), BaseColor=NamedRerouteUsage(), Specular=@n21, Roughness=@n21, Metallic=@n21) {attributeSetTypes [
    #6
    #8
    #9
    #10
   ]}
  NamedRerouteDeclaration(None=StaticSwitchParameter(
    True=SetMaterialAttributes!(MaterialAttributes=@n17, BaseColor=@n18, Specular=@n19, Roughness=@n20) {attributeSetTypes [#6, #8, #9]}
    False=@n22
   ) {group=PixieParts | Textures, parameterName=Use Packed Maps, sortPriority=100}) {name="Stone Material (MA)", nodeColor=<rgba 0.4531 1 0 1>}
  n23 = TextureSampleParameter2D() {
   group=PixieParts | Textures
   parameterName=FX1Mask
   samplerSource=SSM_Wrap_WorldGroupSettings
   samplerType=SAMPLERTYPE_Masks
   sortPriority=10
   texture=&T_BlackMask
  }
  NamedRerouteDeclaration(None=@n23.RGB) {name="FX1 Mask In (V3)", nodeColor=<rgba 0 0.4819 1 1>}
  n24 = TextureSampleParameter2D(UVs=NamedRerouteUsage()) {
   group=PixieParts | Textures
   parameterName=FX2Mask
   samplerSource=SSM_Wrap_WorldGroupSettings
   samplerType=SAMPLERTYPE_Masks
   sortPriority=20
   texture=&T_BlackMask
  }
  NamedRerouteDeclaration(None=@n24.RGB) {name="FX2 Mask In (V3)", nodeColor=<rgba 0 1 0.1524 1>}
  NamedRerouteDeclaration(None=@n16.Distance Blend) {name="Distance Blend (S)", nodeColor=<rgba 0 1 0.7656 1>}
  n25 = TextureSampleParameter2D() {
   group=00_Texture Inputs
   parameterName=Normals
   samplerSource=SSM_Wrap_WorldGroupSettings
   samplerType=SAMPLERTYPE_Normal
   sortPriority=2
   texture=&FlatNormalBlue
  }
  NamedRerouteDeclaration(None=@n25.RGB) {name="Normal Map (V3)", nodeColor=<rgba 0 0.9062 1 1>}
  n26 = TextureSampleParameter2D() {
   group=06_Emissive
   parameterName=Emissive
   samplerSource=SSM_Wrap_WorldGroupSettings
   sortPriority=1
   texture=&T_BlacksRGB
  }
  NamedRerouteDeclaration(None=@n26.RGB) {name="Emissive Map (V3)", nodeColor=<rgba 1 0.8672 0 1>}
  n27 = TextureSampleParameter2D() {
   group=PixieParts | Textures
   parameterName=TechArtMask
   samplerSource=SSM_Wrap_WorldGroupSettings
   samplerType=SAMPLERTYPE_Masks
   sortPriority=0
   texture=&T_BlackMask
  }
  NamedRerouteDeclaration(None=@n27.RGB) {name="FX Mask In (V3)", nodeColor=<rgba 0.5703 0 1 1>}
 ]}
