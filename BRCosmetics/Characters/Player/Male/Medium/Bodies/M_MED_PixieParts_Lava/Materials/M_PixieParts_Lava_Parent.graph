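# Parent character material for the male PixieParts Lava body, using material attributes with a per-expression shading model on skeletal meshes.
# A texture-inputs function publishes base and stone materials, masks, Fresnel, distance blend and pre-skinned position through named reroutes.
# The Enable Customization static switch selects either the customization function or the cockpit material function.
# A shared MF_CharacterParent_2023_PostFX function takes reroute-fed material, mask, position, blend and Fresnel inputs.
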Material /BRCosmetics/Characters/Player/Male/Medium/Bodies/M_MED_PixieParts_Lava/Materials/M_PixieParts_Lava_Parent
graph {
 properties {
  bUseMaterialAttributes=true
  bUsedWithSkeletalMesh=true
  shadingModel=MSM_FromMaterialExpression
  subsurfaceProfile=&/Pearl_SSP
  usedShadingModels=Default Lit | Clear Coat | Subsurface Profile | Cloth
 }
 nodes [
  MaterialFunctionCall(
   Material=NamedRerouteUsage()
   M=NamedRerouteUsage()
   Pre Skinned Local Position=NamedRerouteUsage()
   Distance Blend=NamedRerouteUsage()
   Fresnel=NamedRerouteUsage()
  ) {materialFunction=&MF_CharacterParent_2023_PostFX}
  n1 = MaterialFunctionCall() {materialFunction=/BRCosmetics/Characters/Player/Male/Medium/Bodies/M_MED_PixieParts_Lava/Materials/MaterialFunctions/MF_PixieParts_TextureInputs}
  NamedRerouteDeclaration(None=@n1."Base Material (MA)") {name="Base Material (MA)", nodeColor=<rgba 0.4531 1 0 1>}
  NamedRerouteDeclaration(None=@n1."FX Mask (V3)") {name="FX Mask (V3)", nodeColor=<rgba 0.5703 0 1 1>}
  NamedRerouteDeclaration(None=@n1."FX1 Mask (V3)") {name="FX1 Mask (V3)", nodeColor=<rgba 0 0.4819 1 1>}
  NamedRerouteDeclaration(None=@n1."Distance Blend (S)") {name="Distance Blend (S)", nodeColor=<rgba 0 1 0.7656 1>}
  NamedRerouteDeclaration(None=@n1."Fresnel (S)") {name="Fresnel (S)", nodeColor=<rgba 0 1 0.0156 1>}
  NamedRerouteDeclaration(None=@n1."Mask (V3)") {name="Mask (V3)", nodeColor=<rgba 1 0 0.3047 1>}
  NamedRerouteDeclaration(None=@n1."FX2 Mask (V3)") {name="FX2 Mask (V3)", nodeColor=<rgba 0 1 0.1524 1>}
  n2 = MaterialFunctionCall!(
   Base Material In=NamedRerouteUsage()
   Stone Material In=NamedRerouteUsage()
   Material Masks A=NamedRerouteUsage()
   Material Masks B=NamedRerouteUsage()
   Decal Masks=NamedRerouteUsage()
   AO/GMap/Skin=NamedRerouteUsage()
   Fresnel In=NamedRerouteUsage()
  ) {materialFunction=/BRCosmetics/Characters/Player/Male/Medium/Bodies/M_MED_PixieParts_Lava/Materials/MaterialFunctions/MF_PixieParts_Customization}
  n3 = MaterialFunctionCall!(Material In=NamedRerouteUsage(), TechArt Mask=NamedRerouteUsage(), AO/GMap/Skin=NamedRerouteUsage(), Fresnel In=NamedRerouteUsage()) {materialFunction=/BRCosmetics/Characters/Player/Male/Medium/Bodies/M_MED_PixieParts_Lava/Materials/MaterialFunctions/MF_PixieParts_CockpitMaterial}
  NamedRerouteDeclaration(None=StaticSwitchParameter(True=@n2."Material Out (MA)", False=@n3."Material Out (MA)") {group=PixieParts | Globals, parameterName=Enable Customization, sortPriority=0}) {name="Material Out (MA)", nodeColor=<rgba 0.4531 1 0 1>}
  NamedRerouteDeclaration(None=@n1."Stone Material (MA)") {name="Stone Material (MA)", nodeColor=<rgba 0.4531 1 0 1>}
  NamedRerouteDeclaration(None=@n1."Pre-Skinned Local Position (V3)") {name="Pre-Skinned Local Position (V3)", nodeColor=<rgba 1 0.1641 0 1>}
 ]
}
